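# Additive unlit material for Niagara mesh particles that renders an emissive glow for the Inferno glider.
# It samples the glider's emissive texture, with a "Use RGB" static switch choosing raw or saturated RGB.
# The result is shaped by a Power exponent, boosted by an HDR Color parameter, then multiplied by particle color so the effect system can tint or dim it.
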Material /BRCosmetics/Gadgets/Assets/VinderTech_GliderChute/Glider_Inferno/FX/M_Glider_Inferno_Emissive
graph {
 properties {
  bUsedWithNiagaraMeshParticles=true
  blendMode=BLEND_Additive
  pixelDepthOffsetMode=PDOM_Legacy
  shadingModel=MSM_Unlit
  usedShadingModels=Unlit
 }
 nodes [
  n1 = TextureSample() {texture=/Game/Gadgets/Assets/VinderTech_GliderChute/Glider_Inferno/Textures/T_Glider_Inferno_E}
  n2 = StaticSwitchParameter(True=@n1.RGB, False=Saturate(None=@n1.RGB)) {parameterName=Use RGB}
  n3 = ScalarParameter() {defaultValue=2.7, parameterName=Power}
  Multiply(A=Multiply(A=Power(Base=@n2, Exp=@n3), B=VectorParameter() {parameterName=Color, defaultValue=<rgba 20 20 20 1>}.RGB), B=ParticleColor().RGB)
 ]
}
