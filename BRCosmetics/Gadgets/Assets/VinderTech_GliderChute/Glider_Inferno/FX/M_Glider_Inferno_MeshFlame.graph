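# Translucent, unlit, two-sided Niagara material that draws stylized mesh flames, filed with the Inferno glider FX.
# Emissive blends distorted screen-aligned space textures with a two-color dissolve, plus time-of-day scaling and optional curve-atlas recolor.
# Opacity builds from panned noise, edge masks and sine-based tongue splits, mostly floored for hard edges, with depth fade and optional trail, camera and angle fades.
# Dynamic parameters drive UV offset, dissolve and vertex offsets combining scrolling waves with a clamped bend.
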Material /BRCosmetics/Gadgets/Assets/VinderTech_GliderChute/Glider_Inferno/FX/M_Glider_Inferno_MeshFlame
graph {
 properties {
  bUsedWithNiagaraMeshParticles=true
  bUsedWithNiagaraRibbons=true
  bUsedWithNiagaraSprites=true
  blendMode=BLEND_Translucent
  pixelDepthOffsetMode=PDOM_Legacy
  shadingModel=MSM_Unlit
  twoSided=true
  usedShadingModels=Unlit
 }
 nodes [
  n1 = DynamicParameter() {desc=Parameters set by Niagara, parameterIndex=1, paramNames [X,Y,Z,WPOIntensity]}
  n2 = Clamp(
   None=Subtract(
    A=MaterialFunctionCall(X=@n1.X, Y=@n1.X, Z=@n1.X) {materialFunction=&MakeFloat3}.Result
    B=ParticlePositionWS()
   )
   Min=ScalarParameter() {defaultValue=-100, parameterName=Clamp_min}
   Max=ScalarParameter() {defaultValue=100, parameterName=Clamp_max}
  ) {maxDefault=1000, minDefault=-1000}
  n3 = OneMinus(None=TextureCoordinate())
  n4 = Multiply(A=Power(Base=ComponentMask(None=@n3) {g=true}, Exp=ScalarParameter() {defaultValue=3, parameterName=WPOBendPower}), B=@n1.WPOIntensity)
  n5 = DynamicParameter() {defaultValue=<rgba 1 1 1 100>, paramNames [Offset, Dissolve Progress, WPO Wave Intensity, Sphere Mask Range]}
  n6 = Panner(
   Coordinate=Add(A=@n5.Offset, B=TextureCoordinate() {coordinateIndex=1})
   Time=Multiply(A=Time(), B=ScalarParameter() {defaultValue=1.1, group=Panner Speed, parameterName=Speed})
  ) {speedY=0.8}
  n7 = TextureSample(UVs=@n6) {samplerType=SAMPLERTYPE_Normal, texture=&T_Weap_Distort_N_02}
  n8 = TextureCoordinate()
  n9 = LinearInterpolate(
   B=AppendVector(
    A=ComponentMask(None=@n7.RGB) {g=true, r=true}
    B=TextureSample(UVs=@n6) {texture=&T_Smooth_tile}.R
   )
   Alpha=OneMinus(None=ComponentMask(None=@n8) {g=true})
  )
  n10 = DynamicParameter() {defaultValue=<rgba 1 1 1 100>, paramNames [Offset, Dissolve Progress, WPO Wave Intensity, Sphere Mask Range]}
  Add(
   A=LinearInterpolate(A=Constant(), B=LinearInterpolate(B=@n2, Alpha=@n4), Alpha=ScalarParameter() {defaultValue=1, parameterName=Use interpolate WPO})
   B=LinearInterpolate(
    A=Multiply(A=@n9, B=@n10.WPO Wave Intensity) {constB=10}
    Alpha=ScalarParameter() {desc=no_Wpo, group=Simple_Round_particle, parameterName=No_WPO}
   )
  )
  n11 = TextureCoordinate()
  n12 = VectorParameter() {group=Fire Shape, parameterName=Fire_MacroShape_UVScale, sortPriority=5, defaultValue=<rgba 1 1 0 0>}
  n13 = ComponentMask(None=@n12.RGB) {g=true, r=true}
  n14 = MaterialFunctionCall(Float2=Multiply(A=@n11, B=@n13)) {materialFunction=&BreakOutFloat2Components}
  n15 = Power(Base=@n14.G, Exp=ScalarParameter() {defaultValue=2, group=Fire Shape, parameterName=Fire Split Contrast, sortPriority=1})
  n16 = Multiply(A=@n14.R) {constB=0.45}
  n17 = DynamicParameter() {defaultValue=<rgba 1 1 1 100>, paramNames [Offset, Dissolve Progress, WPO Wave Intensity, Sphere Mask Range]}
  n18 = Add(A=@n17.Offset, B=TextureCoordinate() {coordinateIndex=1, vTiling=1.5})
  n19 = ComponentMask(None=VectorParameter() {group=Distortion, parameterName=Distortion UV Scale, sortPriority=3, defaultValue=<rgba 1 1 0 0>}.RGB) {g=true, r=true}
  n20 = Time()
  n21 = ScalarParameter() {defaultValue=1, group=Panner Speed, parameterName=Speed}
  n22 = TextureSampleParameter2D(UVs=Panner(Coordinate=Multiply(A=@n18, B=@n19), Time=Multiply(A=@n20, B=@n21)) {speedY=1}) {
   group=Distortion
   parameterName=Distortion Normal
   samplerType=SAMPLERTYPE_Normal
   sortPriority=2
   texture=&T_Weap_Distort_N_02
  }
  n23 = Power(Base=OneMinus(None=Sine(None=@n16)), Exp=LinearInterpolate(Alpha=@n22.R) {constA=0.5, constB=1.5}) {constExponent=0.5}
  LinearInterpolate(B=Round(None=Add(A=@n15, B=@n23)), Alpha=ScalarParameter() {defaultValue=1, group=Fire Shape, parameterName=Fire Split}) {constA=1, constB=0}
  n24 = MaterialFunctionCall!(TextureResolution=ScalarParameter() {defaultValue=256, parameterName=NebulaTextureResolution}) {materialFunction=/Engine/Functions/Engine_MaterialFunctions02/Texturing/ScreenAlignedPixelToPixelUVs}
  n25 = ComponentMask(None=VectorParameter() {parameterName="Nebula&&StarfieldPannerSpeed", defaultValue=<rgba 0 0.05 0 0>}.RGB) {g=true, r=true}
  n26 = TextureSampleParameter2D!(UVs=Panner(Coordinate=@n24.Result, Speed=@n25), Level, Apply View MipBias) {
   mipValueMode=TMVM_MipLevel
   parameterName=Space_Tex_A
   texture=/Game/Characters/Player/Male/Medium/Bodies/M_Med_Soldier_04/Skins/Galaxy/Textures/T_FN_Nebula
  }
  n27 = Desaturation(None=Multiply(A=@n26.RGB), Fraction=ScalarParameter() {defaultValue=-0.5, parameterName=Space_Tex_A_desaturation})
  n28 = ComponentMask(None=@n22.RGB) {g=true, r=true}
  n29 = ComponentMask(None=Multiply(A=@n11, B=@n13)) {g=true}
  n30 = LinearInterpolate(
   B=Divide(
    A=LinearInterpolate(A=@n28, Alpha=@n29) {constB=0}
    B=ScalarParameter() {defaultValue=3, group=Distortion, parameterName=UV_Distortion_Divsor, sortPriority=1}
   ) {constB=20}
   Alpha=ScalarParameter() {defaultValue=1, group=Distortion, parameterName=UV_Distortion_Lerp}
  )
  n31 = DynamicParameter() {defaultValue=<rgba 1 1 1 100>, paramNames [Offset, Dissolve Progress, WPO Wave Intensity, Sphere Mask Range]}
  n32 = Add(A=@n31.Offset, B=TextureCoordinate() {coordinateIndex=1, uTiling=0.8, vTiling=0.9})
  n33 = ComponentMask(None=VectorParameter() {group=Fire Shape, parameterName=Fire_Micro_UVScale, sortPriority=8, defaultValue=<rgba 1 1 0 0>}.RGB) {g=true, r=true}
  n34 = Panner(
   Coordinate=MaterialFunctionCall(XY=Multiply(A=@n32, B=@n33)) {materialFunction=&Swizzle}.YX
   Time=Multiply(A=Time(), B=ScalarParameter() {defaultValue=1, group=Panner Speed, parameterName=Speed})
  ) {speedX=1.2, speedY=0.2}
  n35 = TextureSampleParameter2D!(UVs=ShadingPathSwitch(Default=Add(A=@n30, B=@n34), Mobile=@n34) {desc=Migrated from FeatureLevelSwitch}, Apply View MipBias) {
   group=Fire Shape
   parameterName=Fire Micro Texture
   samplerType=SAMPLERTYPE_LinearColor
   sortPriority=6
   texture=&T_NoisePack_LinearColor
  }
  n36 = ComponentMask(None=@n35.RGB) {g=true, r=true}
  n37 = Multiply(A=@n36) {constB=0.05}
  n38 = TextureSampleParameter2D!(UVs=Panner(Coordinate=Add(A=@n24.Result, B=@n37), Speed=@n25), Level=Constant() {r=1}, Apply View MipBias) {mipValueMode=TMVM_MipLevel, parameterName=Space_Tex_B, texture=/Game/Effects/Fort_Effects/Textures/Plasma/T_Plasma_Aurora_Seamless}
  n39 = Add(A=@n27, B=Desaturation(None=@n38.RGB, Fraction=ScalarParameter() {defaultValue=-1, parameterName=Des tweak}))
  n40 = ScalarParameter() {group=Simple_Round_particle, parameterName=Only_Use_one_Space_Tex}
  n41 = LinearInterpolate(
   A=Multiply(
    A=LinearInterpolate(A=@n39, B=@n27, Alpha=@n40)
    B=ScalarParameter() {defaultValue=1, group=Simple_Round_particle, parameterName=Texture_intensity}
   )
   B=Constant()
   Alpha=ScalarParameter() {parameterName=Use simple texture}
  )
  n42 = OneMinus(None=ComponentMask(None=TextureCoordinate()) {g=true})
  n43 = Saturate(None=Power(Base=Add(A=@n42) {constB=-0.2}, Exp=ScalarParameter() {defaultValue=3, parameterName=Space_Texture_center_mask_power}) {constExponent=0.5})
  n44 = TextureCoordinate() {uTiling=0.7, vTiling=0.8}
  n45 = Multiply(A=@n36, B=ScalarParameter() {group=Dissolution, parameterName=Dissolution Texture UV Distortion Strength})
  n46 = DynamicParameter() {defaultValue=<rgba 1 1 1 100>, paramNames [Offset, Dissolve Progress, WPO Wave Intensity, Sphere Mask Range]}
  n47 = Time()
  n48 = ScalarParameter() {group=Dissolution, parameterName=Loop_panner}
  n49 = TextureSampleParameter2D!(
   UVs=Panner(Coordinate=Add(A=@n44, B=@n45), Time=LinearInterpolate(A=@n46.Dissolve Progress, B=@n47, Alpha=@n48)) {speedY=0.2}
   Apply View MipBias
  ) {group=Dissolution, parameterName=dissolution Texture, texture=&T_Inky_Smoke_Tile}
  n50 = Multiply(A=@n14.R) {constB=0.5}
  n51 = ScalarParameter() {defaultValue=6, group=Fire Shape, parameterName=Edge Fade Value Top, sortPriority=3}
  n52 = ScalarParameter() {defaultValue=2, group=Fire Shape, parameterName=Edge Fade Value Bottom, sortPriority=4}
  n53 = Multiply(A=Power(Base=Sine(None=@n50), Exp=LinearInterpolate(A=@n51, B=@n52, Alpha=@n14.G) {constA=6, constB=2}) {constExponent=4}, B=@n14.G)
  n54 = Multiply(
   A=Power(Base=@n53, Exp=ScalarParameter() {defaultValue=1, group=Dissolution, parameterName=Hot Spot Mask Exp})
   B=ScalarParameter() {defaultValue=1, group=Dissolution, parameterName=Hot Spot Mask Multiply}
  )
  n55 = LinearInterpolate(
   A=VectorParameter() {group=Dissolution, parameterName=Color1, defaultValue=<rgba 0.0015 0.0028 0.0104 1>}.RGB
   B=VectorParameter() {group=Dissolution, parameterName=Color2, defaultValue=<rgba 0.2135 0.0266 0.1457 1>}.RGB
   Alpha=@n49.R
  )
  n56 = Multiply(
   A=LinearInterpolate(
    A=VectorParameter() {group=Dissolution, parameterName=Hot Spot Mask Low Color, defaultValue=<rgba 1 1 1 1>}.RGB
    B=VectorParameter() {group=Dissolution, parameterName=Hot Spot Mask High Color, defaultValue=<rgba 1 1 1 1>}.RGB
    Alpha=Clamp(None=@n54)
   )
   B=@n55
  )
  n57 = ScalarParameter() {group=Dissolution, parameterName=Use Mask Hot Spot}
  n58 = TextureCoordinate()
  n59 = Constant2Vector() {g=0.5, r=0.5}
  n60 = Multiply(
   A=LinearInterpolate(A=LinearInterpolate(A=@n55, B=@n56, Alpha=@n57), B=Multiply(A=@n41, B=@n43), Alpha=@n49.R)
   B=Power(
    Base=SphereMask(A=@n58, B=@n59) {attenuationRadius=0.8, hardnessPercent=0}
    Exp=ScalarParameter() {defaultValue=3, parameterName=Space_Texture_power}
   ) {constExponent=3}
  )
  n61 = Add(A=ComponentMask(None=TextureCoordinate()) {g=true}) {constB=-0.72}
  n62 = Power(Base=Saturate(None=Multiply(A=@n61))) {constExponent=1}
  n63 = LinearInterpolate(
   A=LinearInterpolate(Alpha=Saturate(None=@n62)) {constA=1, constB=7}
   B=Constant() {r=1}
   Alpha=ScalarParameter() {group=Global Gradient, parameterName=Not_Use_Global_Gradient}
  )
  n64 = ParticleColor()
  n65 = Multiply(
   A=MaterialFunctionCall(
    Value_Morning=ScalarParameter() {defaultValue=1, group=Time of Day, parameterName=Color Value Morning}
    Value_Day=ScalarParameter() {defaultValue=1, group=Time of Day, parameterName=Color Value Day}
    Value_Evening=ScalarParameter() {defaultValue=1, group=Time of Day, parameterName=Color Value Evening}
    Value_Night=ScalarParameter() {defaultValue=1, group=Time of Day, parameterName=Color Value Night}
   ) {materialFunction=&Scalar_Time_of_Day}.Result
   B=Multiply(A=Multiply(A=@n60, B=@n63), B=@n64.RGB)
  )
  n66 = Panner(Speed=ComponentMask(None=VectorParameter() {parameterName=Param_1, defaultValue=<rgba 0 0.5 0 1>}.RGB) {g=true, r=true})
  n67 = ScalarParameter() {defaultValue=1, parameterName=Param_2}
  n68 = Add(
   A=LinearInterpolate(
    A=TextureCoordinate()
    B=TextureSample(UVs=@n66) {texture=/Game/Global/Textures/Default/Noise/T_FireNoise_01}.R
    Alpha=Multiply(A=@n67) {constB=0.1}
   )
   B=ComponentMask(None=VectorParameter() {defaultValue=<rgba 0 -0.25 0 1>}.RGB) {g=true, r=true}
  )
  n69 = CurveAtlasRowParameter(CurveTime=LinearInterpolate(
    A=VertexColor().R
    B=TextureSample(UVs=@n68) {texture=/Game/Accessories/FORT_Backpacks/Backpack_F_MED_Ashes/FX/T_FX_Backpack_Ashes}.R
    Alpha=ScalarParameter() {defaultValue=0.55, parameterName=Param_1}
   )) {
   atlas=/Game/Effects/Niagara/MasterMaterials/Smoke/Textures/CurveAtlas
   curve=/Game/Effects/Niagara/MasterMaterials/Smoke/Textures/EmissiveCurve
   defaultValue=1
   parameterName=ColorCurve
  }
  n70 = Desaturation(None=@n65, Fraction=ScalarParameter() {defaultValue=1})
  MaterialFunctionCall(Emissive=LinearInterpolate(A=@n65, B=Multiply(A=@n69, B=@n70), Alpha=ScalarParameter() {parameterName=UseColorCurve})) {materialFunction=&MF_FX_EmissiveAdjustments}
  Panner()
  n71 = ParticleColor()
  n72 = Panner(Coordinate=TextureCoordinate(), Speed=AppendVector(A=ScalarParameter() {defaultValue=0.4, parameterName=Speed1}, B=Constant()))
  n73 = TextureCoordinate() {vTiling=0}
  n74 = Constant2Vector() {r=0.5}
  n75 = DepthFade()
  n76 = Multiply(
   A=Multiply(
    A=TextureSample(UVs=@n72) {samplerType=SAMPLERTYPE_Grayscale, texture=&T_Smoke_Trail_Soft}.R
    B=SphereMask(A=@n73, B=@n74) {attenuationRadius=0.5, hardnessPercent=0}
   )
   B=ShadingPathSwitch(Default=Multiply(A=@n75), Mobile=Constant() {r=1}) {desc=Migrated from FeatureLevelSwitch}
  )
  n77 = Multiply(A=@n64.A, B=SphereMask(A=TextureCoordinate(), B=Constant2Vector() {g=0.5, r=0.5}, Radius=Constant() {r=0.5}) {hardnessPercent=70})
  n78 = ScalarParameter() {defaultValue=50, parameterName=Round_mask_depth_fade}
  n79 = TextureSample() {samplerType=SAMPLERTYPE_LinearGrayscale, texture=/Engine/Functions/Engine_MaterialFunctions02/ExampleContent/Textures/LowResBlurredNoise}
  n80 = VectorParameter() {group=BottomFade, parameterName=Bottom Fade_SphereMask_A_Scale, sortPriority=2, defaultValue=<rgba 1.2 1 0 0>}
  n81 = Multiply(A=TextureCoordinate(), B=ComponentMask(None=@n80.RGB) {g=true, r=true})
  n82 = VectorParameter() {group=BottomFade, parameterName=Bottom Fade_SphereMask_B, sortPriority=3, defaultValue=<rgba 0.6 0.05 0 0>}
  n83 = ComponentMask(None=@n82.RGB) {g=true, r=true}
  n84 = ScalarParameter() {defaultValue=0.95, group=BottomFade, parameterName=BottomFade_Radius, sortPriority=1}
  n85 = Saturate(None=Multiply(A=SphereMask(A=@n81, B=@n83, Radius=@n84) {attenuationRadius=0.95, hardnessPercent=0}) {constB=15})
  n86 = ScalarParameter() {defaultValue=1, group=BottomFade, parameterName=Bottom Fade}
  n87 = Multiply(
   A=LinearInterpolate(
    A=ChannelMaskParameter(None=@n35.RGBA) {group=Fire Shape, parameterName=Fire Micro Mask, sortPriority=7}.R
    B=@n79.R
    Alpha=ScalarParameter() {defaultValue=0.2, parameterName=Fire Micro lerp}
   ) {constAlpha=0.2}
   B=Multiply(A=@n53, B=LinearInterpolate(B=@n85, Alpha=@n86) {constA=1})
  )
  n88 = LinearInterpolate(Alpha=@n14.G) {constA=2}
  n89 = MaterialFunctionCall(Float2=TextureCoordinate()) {materialFunction=&BreakOutFloat2Components}
  n90 = Multiply(A=ScalarParameter() {defaultValue=1.4, group=Frame Split, parameterName=Frame Split Offset A}, B=@n89.R) {constB=1.5}
  n91 = Multiply(A=OneMinus(None=Sine(None=@n90) {period=2}), B=ScalarParameter() {defaultValue=5, group=Frame Split, parameterName=Frame Split Thin A})
  n92 = DynamicParameter() {defaultValue=<rgba 1 1 1 100>, paramNames [Offset, Dissolve Progress, WPO Wave Intensity, Sphere Mask Range]}
  n93 = VectorParameter() {group=Frame Split, parameterName=Frame Split Distortion UV, defaultValue=<rgba 1 2 0 1>}
  n94 = TextureCoordinate()
  n95 = Multiply(A=@n94, B=AppendVector(A=@n93.R, B=@n93.R))
  n96 = AppendVector(A=@n93.B, B=@n93.B)
  n97 = TextureSample(UVs=Add(A=AppendVector(A=@n92.Offset, B=@n92.Offset), B=Panner(Coordinate=@n95, Speed=@n96))) {samplerType=SAMPLERTYPE_Normal, texture=&T_Weap_Distort_N_02}
  n98 = ScalarParameter() {defaultValue=1.5, group=Frame Split, parameterName=Frame Split Distortion Intensity}
  n99 = ComponentMask(None=TextureCoordinate()) {g=true}
  n100 = LinearInterpolate(Alpha=Multiply(A=Multiply(A=@n97.R, B=@n98), B=OneMinus(None=@n99))) {constA=0.5, constB=1.5}
  n101 = ScalarParameter() {defaultValue=4, group=Frame Split, parameterName=Frame Split Depth A}
  n102 = Round(None=Add(A=Power(Base=@n91, Exp=@n100) {constExponent=0.5}, B=Power(Base=@n89.G, Exp=@n101)))
  n103 = Multiply(A=ScalarParameter() {defaultValue=0.8, group=Frame Split, parameterName=Frame Split Offset B}, B=@n89.R) {constB=1.5}
  n104 = Multiply(A=OneMinus(None=Sine(None=@n103) {period=2}), B=ScalarParameter() {defaultValue=15, group=Frame Split, parameterName=Frame Split Thin B})
  n105 = ScalarParameter() {defaultValue=3, group=Frame Split, parameterName=Frame Split Depth B}
  n106 = Round(None=Add(A=Power(Base=@n104, Exp=@n100) {constExponent=0.5}, B=Power(Base=@n89.G, Exp=@n105)))
  n107 = LinearInterpolate(
   B=Multiply(A=Saturate(None=@n102), B=Saturate(None=@n106))
   Alpha=ScalarParameter() {defaultValue=0.7, group=Frame Split, parameterName=Lerp Frame Split}
  ) {constA=1, constB=0}
  n108 = VectorParameter() {group=Frame Detail, parameterName=Frame Detail Distortion UV, defaultValue=<rgba 4 0.5 -0.25 0.75>}
  n109 = Multiply(A=@n94, B=AppendVector(A=@n108.R, B=@n108.R))
  n110 = AppendVector(A=@n108.B, B=@n108.B)
  n111 = TextureSample(UVs=Add(A=AppendVector(A=@n92.Offset, B=@n92.Offset), B=Panner(Coordinate=@n109, Speed=@n110))) {samplerType=SAMPLERTYPE_Normal, texture=&T_Weap_Distort_N_02}
  n112 = ScalarParameter() {defaultValue=0.008, group=Frame Detail, parameterName=Frame Detail Distortion Intensity}
  n113 = ComponentMask(None=TextureCoordinate()) {g=true}
  n114 = TextureSampleParameter2D(UVs=Multiply(A=Multiply(A=@n111.G, B=@n112), B=OneMinus(None=@n113))) {
   group=Frame Detail
   parameterName=Frame Detail Tex
   texture=/Game/Gadgets/Assets/VinderTech_GliderChute/Glider_Headset/FX/T_Glider_Headset_Mask_01
  }
  n115 = Power(Base=@n114.G, Exp=ScalarParameter() {defaultValue=1, group=Frame Detail, parameterName=Frame Detail Power})
  n116 = ScalarParameter() {defaultValue=8, group=Frame Detail, parameterName=Frame Detail Mult}
  n117 = LinearInterpolate(
   B=Saturate(None=Multiply(A=@n115, B=@n116) {constB=8})
   Alpha=ScalarParameter() {defaultValue=0.7, group=Frame Detail, parameterName=Lerp Frame Detail}
  ) {constA=1, constB=0}
  n118 = Multiply(
   A=Multiply(A=Power(Base=@n87, Exp=@n88), B=Multiply(A=@n107, B=@n117))
   B=ScalarParameter() {defaultValue=15, group=Opacity, parameterName=OpacityMaskMultiplier, sortPriority=1}
  ) {constB=3}
  n119 = LinearInterpolate(A=Floor(None=@n118), B=@n118, Alpha=ScalarParameter() {defaultValue=0.06, parameterName=Floor})
  n120 = Multiply(A=DepthFade(Opacity=@n119, FadeDistance=ScalarParameter() {defaultValue=200, group=Opacity, parameterName=DepthFadeDistance}))
  n121 = ScalarParameter() {defaultValue=1, parameterName=Multiply by Particle Alpha}
  n122 = StaticSwitchParameter(
   True=LinearInterpolate(
    A=Multiply(A=@n71.A, B=@n76)
    B=DepthFade(Opacity=@n77, FadeDistance=@n78)
    Alpha=ScalarParameter() {parameterName=Use_simple_round_particle_mask}
   )
   False=Multiply(
    A=ShadingPathSwitch(Default=@n120, Mobile=@n119) {desc=Migrated from FeatureLevelSwitch}
    B=LinearInterpolate(B=@n64.A, Alpha=@n121) {constA=1}
   )
  ) {parameterName=Pickaxe_Swing_Mask}
  n123 = ScalarParameter() {defaultValue=150, group=Camera Fade, parameterName=Near Camera Distance}
  n124 = LinearInterpolate(
   A=@n122
   B=Multiply(A=@n122, B=MaterialFunctionCall(Distance=@n123) {materialFunction=&Near_Camera_Fade}.Result)
   Alpha=ScalarParameter() {group=Camera Fade, parameterName=Use Camera Fade}
  )
  n125 = ScalarParameter() {defaultValue=2, group=Ancient Gladiator, parameterName=Final Opacity Power A}
  n126 = ScalarParameter() {defaultValue=1, group=Ancient Gladiator, parameterName=Final Opacity Power B}
  n127 = ComponentMask(None=TextureCoordinate()) {g=true}
  n128 = Multiply(A=@n127, B=ScalarParameter() {defaultValue=3, group=Ancient Gladiator, parameterName=Final Opacity Gradient Hardness})
  n129 = Multiply(A=Power(Base=@n124, Exp=LinearInterpolate(A=@n125, B=@n126, Alpha=@n127)), B=Power(Base=Clamp(None=@n128)) {constExponent=1})
  n130 = ScalarParameter() {group=Ancient Gladiator, parameterName=Use Ancient Gladiator Mask}
  n131 = ScalarParameter() {parameterName=FadeBasedOnViewAngle}
  Reroute(None=Multiply(
    A=LinearInterpolate(A=@n124, B=@n129, Alpha=@n130)
    B=MaterialFunctionCall("FadeSensitivity (optional)"=@n131) {materialFunction=&FadeBasedOnViewAngle}.Result
   ))
  ScalarParameter() {defaultValue=30, group=WPO, parameterName=WaveIntensity, sortPriority=7}
 ]
}
