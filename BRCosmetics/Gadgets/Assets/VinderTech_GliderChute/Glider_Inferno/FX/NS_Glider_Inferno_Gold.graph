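# Niagara effect for the gold Inferno glider: engine flames, flares, lights, emissive and flame body meshes, and hair, attached to the glider's skeletal mesh bones or sockets.
# Looping flames and flares stretch and brighten with a system FN_Integrator value, paired with one-shot ignition bursts after User.EngineDelay; User.bIsFullyDeployed gates flames and lights.
# Front-end preview and time-of-day modifiers retune color and intensity.
# Flame and hair meshes feed an interpolated position into particle attributes for trailing motion.
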NiagaraSystem /BRCosmetics/Gadgets/Assets/VinderTech_GliderChute/Glider_Inferno/FX/NS_Glider_Inferno_Gold
dependencies {
 usedDataInterfaces [
  &NiagaraDataInterfaceArrayFloat3
  &NiagaraDataInterfaceArrayInt32
  &NiagaraDataInterfaceCurve
  &NiagaraDataInterfaceArrayFloat
  &NiagaraDataInterfaceVector2DCurve
  &NiagaraDataInterfaceColorCurve
  &NiagaraDataInterfaceCamera
 ]
 usedDynamicInputs [&SetBoolByFloatComparison, &Abs_Float]
 usedModules [
  &self:SystemSpawnScript.SystemScriptSource.SystemScriptGraph.NiagaraNodeAssignment_0.SetVariables_#1
  &SystemState
  &FN_Integrator
 ]
}
emitters {
 EmissiveMesh {
  emitterUpdateScript [
   &EmitterState(Life Cycle Mode=::NewEnumerator1, Loop Behavior=::NewEnumerator1, Loop Duration Mode=::NewEnumerator1, UseLoopDelay=false)
   &SpawnBurst_Instantaneous(Use Spawn Probability=false, UseLoopCountLimit=false)
  ]
  particleSpawnScript [
   &InitializeParticle(Mesh Scale Mode=::NewEnumerator3, UsePositionOffset=false, Write Material Random=true)
   &InitialMeshOrientation(
    Mesh Orientation Mode=::NewEnumerator2
    Rotation=&FN_FrontEndVector3Modifier(Gameplay=<0 0 0>, Front End=<0 0 0>)
    Use Rotation Vector=true
   )
   &CameraOffset(Camera Offset Amount=1)
  ]
  particleUpdateScript [
   &ParticleState(Kill Particles When Lifetime Has Elapsed=false, Let Infinitely Lived Particles Die When Emitter Deactivates=true)
   &ScaleColor(
    Scale Alpha=&Multiply_Float(
     A=&FloatFromCurve(FloatCurve=DataInterface() {propertyValues {Curve {keys [<at 0 0>,<at 0.4 1>]}}}, CurveIndex=$System.Age)
     B=1
    )
    Scale RGB=&VectorFromFloat(Value=&FloatFromCurve(
      FloatCurve=DataInterface() {propertyValues {Curve {keys [
          {interpMode=RCIM_Cubic, tangentMode=RCTM_User, time=0.1, value=0.2}
          {interpMode=RCIM_Cubic, tangentMode=RCTM_User, time=0.25, value=3}
          <at 0.5 1>
         ]}}}
      CurveIndex=$System.Age
      Scale Curve=&FN_FrontEndFloatModifier(Gameplay=&FNTimeOfDayScalar(Night Scale=3, Morning Scale=1, Evening Scale=1), Front End=0.7)
     ))
    ScaleA=true
    ScaleRGB=true
   )
   &SkeletalMeshLocation(
    Apply Sampled Orientation to Mesh Particles=true
    Bone / Socket Index=3
    Inherit Velocity=false
    Interpolate Mesh Sampling=false
    MaskSpawnGroup=false
    Mesh Sampling Type=::NewEnumerator5
    Sampled Position Offset Coordinate Space:ENiagaraMeshTransforms=::NewEnumerator3
    Skeletal Mesh=$User.SkeletalMesh
    Socket Sampling Mode=::NewEnumerator5
    Use Sampled Mesh Attribute as Ribbon ID=false
   )
  ]
  rendererClasses [&NiagaraMeshRendererProperties]
  renderers [{rendererClass=&NiagaraMeshRendererProperties, rendererIndex=0}]
 }
 EngineFlames {
  emitterUpdateScript [
   &EmitterState(Life Cycle Mode=::NewEnumerator1, Loop Behavior=::NewEnumerator1, Loop Delay=$User.EngineDelay, Loop Duration=1, UseLoopDelay=true)
   &SpawnBurst_Instantaneous(Spawn Count=2, Use Spawn Probability=true)
  ]
  particleSpawnScript [
   &FN_SetSpawnGroup(Number of Spawn Groups=2)
   &InitializeParticle(
    Color=&MakeCustomLinearColorFromBool(Bool=$User.bIsFrontEndPreview, True Color=<rgba 0.6599 1 0.845 1>, False Color=<rgba 1 1 1 1>)
    Lifetime=0.25
    Mesh Scale=&SelectVectorFromArray(
     Direct Array Index=$Particles.SpawnGroup
     Vector Selection Array=DataInterface() {propertyValues {FloatData [<0.35 0.35 0.25>, <0.35 0.35 0.25>, <0.65 0.65 0.65>]}}
    )
    Mesh Scale Mode=::NewEnumerator4
    Sprite Size Mode=::NewEnumerator3
    Uniform Sprite Size=220
    UsePositionOffset=false
    Write Material Random=true
   )
   &InitialMeshOrientation(Rotation=<0 0.25 0>, Use Orientation Vector=true, Use Rotation Vector=true) {module=InitialMeshOrientation001}
   &InitialMeshOrientation(
    Mesh Orientation Mode=::NewEnumerator1
    Rotation=&RandomRangeVector(Minimum=<-1 0 0>, Maximum=<1 0 0>)
    Use Rotation Vector=true
   )
   /SetVariables(
    Particles.Flame2Scale:Vector3f=&Multiply_Vector(A=$Particles.Scale, B=<1 1 0.9>)
    Particles.FlameCoreScale:Vector3f=&Multiply_Vector(A=$Particles.Scale, B=<0.65 0.6 0.6>)
   ) {isSetParameters=true, module=SetVariables_#2}
  ]
  particleUpdateScript [
   &ParticleState(Kill Particles When Lifetime Has Elapsed=false)
   &SkeletalMeshLocation(
    Apply Sampled Orientation to Mesh Particles=true
    Bone / Socket Index=&SelectIntFromArray(Direct Array Index=$Particles.SpawnGroup, Int Selection Array=DataInterface() {propertyValues {IntData [0,1,2]}})
    Inherit Velocity=false
    MaskSpawnGroup=false
    Mesh Sampling Type=::NewEnumerator5
    Skeletal Mesh=$User.SkeletalMesh
    Socket Sampling Mode=::NewEnumerator5
   )
   &SolveForcesAndVelocity(Clamp Velocity=false, Limit Acceleration=false)
   &ScaleColor(
    Scale Alpha=&FloatFromCurve(
     FloatCurve=DataInterface() {propertyValues {Curve {keys [<at 0 0 RCIM_Cubic>,<at 0.5 1 RCIM_Cubic>]}}}
     Scale Curve=&Multiply_Float(
      A=&Lerp_Float(
       A=0.15
       B=1
       Alpha=&FN_FrontEndFloatModifier(Gameplay=$System.FN_Integrator001.Value, Front End=$User.FE Intergrator Value)
      )
      B=&MakeCustomFloatFromBool(Bool=$User.bIsFullyDeployed, True Float=1, False Float=0)
     )
    )
    Scale RGB=&Multiply_VectorByFloat(
     Vector=&Multiply_VectorByFloat(Vector=<1 1 1>, Float=&FNTimeOfDayScalar(Night Scale=1.5, Morning Scale=1, Evening Scale=1))
     Float=&FN_FrontEndFloatModifier(Gameplay=0.7, Front End=0.5)
    )
    ScaleA=true
    ScaleRGB=true
   )
   /SetVariables(
    Particles.Flame2Color:LinearColor=&MakeLinearColorFromVectorAndFloat(
     "Vector (RGB)"=&Multiply_VectorByFloat(Vector=<1 1 1>, Float=&FN_FrontEndFloatModifier(Gameplay=1, Front End=1))
     "Float (Alpha)"=&Multiply_Float(
      A=&Multiply_Float(
       A=&Lerp_Float(
        A=0.3
        B=0.5
        Alpha=&FN_FrontEndFloatModifier(Gameplay=$System.FN_Integrator001.Value, Front End=$User.FE Intergrator Value)
       )
       B=&MakeCustomFloatFromBool(Bool=$User.bIsFullyDeployed, True Float=6, False Float=0)
      )
      B=&FN_FrontEndFloatModifier(Gameplay=1, Front End=0.3)
     )
    )
    Particles.FlameCoreColor:LinearColor=&MakeLinearColorFromVectorAndFloat(
     "Vector (RGB)"=&MakeVectorFromLinearColorRGB(Color=&MultiplyLinearColor(
       Linear Color=<rgba 10 4.7464 1.0514 1>
       Scale Factor=&MakeVector4_FromFloat(Float=&FN_FrontEndFloatModifier(Gameplay=2, Front End=1))
      ))
     "Float (Alpha)"=&Multiply_Float(
      A=&Lerp_Float(
       A=0.1
       B=0.3
       Alpha=&FN_FrontEndFloatModifier(Gameplay=$System.FN_Integrator001.Value, Front End=$User.FE Intergrator Value)
      )
      B=&MakeCustomFloatFromBool(Bool=$User.bIsFullyDeployed, True Float=1, False Float=0)
     )
    )
   ) {isSetParameters=true, module=SetVariables_#3}
   &ScaleMeshSize(Scale Factor=&MakeVector(
     X=1
     Y=1
     Z=&FloatFromCurve(FloatCurve=DataInterface() {propertyValues {Curve {keys [<at 0 0>,<at 0.85 1>]}}})
    ))
   &ScaleMeshSize(Scale Factor=&MakeVector(
     X=1
     Y=1
     Z=&Lerp_Float(
      A=1
      B=&SelectFloatFromArray(Direct Array Index=$Particles.SpawnGroup, Float Selection Array=DataInterface() {propertyValues {FloatData [2.4,2.4,2]}})
      Alpha=&FN_FrontEndFloatModifier(Gameplay=$System.FN_Integrator001.Value, Front End=$User.FE Intergrator Value)
     )
    )) {module=ScaleMeshSize001}
   /SetVariables(
    Particles.Flame2Scale:Vector3f=&Multiply_Vector(
     A=$Particles.Initial.Flame2Scale
     B=&MakeVector(
      X=&FloatFromCurve(FloatCurve=DataInterface() {propertyValues {Curve {keys [<at 0 0>,<at 1 1>]}}})
      Y=1
      Z=1
     )
    )
    Particles.FlameCoreScale:Vector3f=&Multiply_Vector(
     A=$Particles.Initial.FlameCoreScale
     B=&MakeVector(
      X=&FloatFromCurve(FloatCurve=DataInterface() {propertyValues {Curve {keys [<at 0 0>,<at 1 1>]}}})
      Y=1
      Z=1
     )
    )
   ) {isSetParameters=true, module=SetVariables_#4}
   /SetVariables(
    Particles.Flame2Scale:Vector3f=&Multiply_Vector(
     A=$Particles.Flame2Scale
     B=&MakeVector(
      X=1
      Y=1
      Z=&Lerp_Float(
       A=&SelectFloatFromArray(Direct Array Index=$Particles.SpawnGroup, Float Selection Array=DataInterface() {propertyValues {FloatData [1.3,1.3,1.2]}})
       B=&SelectFloatFromArray(Direct Array Index=$Particles.SpawnGroup, Float Selection Array=DataInterface() {propertyValues {FloatData [2.2,2.2,1.7]}})
       Alpha=&FN_FrontEndFloatModifier(Gameplay=$System.FN_Integrator001.Value, Front End=$User.FE Intergrator Value)
      )
     )
    )
    Particles.FlameCoreScale:Vector3f=&Multiply_Vector(
     A=$Particles.FlameCoreScale
     B=&MakeVector(
      X=1
      Y=1
      Z=&Lerp_Float(
       A=&SelectFloatFromArray(Direct Array Index=$Particles.SpawnGroup, Float Selection Array=DataInterface() {propertyValues {FloatData [1,1,1.25]}})
       B=&SelectFloatFromArray(Direct Array Index=$Particles.SpawnGroup, Float Selection Array=DataInterface() {propertyValues {FloatData [1.7,1.7,1.75]}})
       Alpha=&FN_FrontEndFloatModifier(Gameplay=$System.FN_Integrator001.Value, Front End=$User.FE Intergrator Value)
      )
     )
    )
   ) {isSetParameters=true, module=SetVariables_#5}
   &ScaleSpriteSize()
   &DynamicMaterialParameters(
    Index 0 Param 1=&Lerp_Float(
     A=0.3
     B=0.6
     Alpha=&FN_FrontEndFloatModifier(Gameplay=$System.FN_Integrator001.Value, Front End=$User.FE Intergrator Value)
    )
    Module.Param0WriteEnabled=true
    Module.Param1WriteEnabled=true
    Module.Param2WriteEnabled=true
    Module.Param3WriteEnabled=true
   )
  ]
  rendererClasses [&NiagaraMeshRendererProperties]
  renderers [
   {rendererClass=&NiagaraMeshRendererProperties, rendererIndex=0}
   {rendererClass=&NiagaraMeshRendererProperties, rendererIndex=1}
   {rendererClass=&NiagaraMeshRendererProperties, rendererIndex=2}
   {rendererClass=&NiagaraMeshRendererProperties, rendererIndex=3}
  ]
 }
 EngineFlames001 {
  emitterUpdateScript [
   &EmitterState(Life Cycle Mode=::NewEnumerator1, Loop Behavior=::NewEnumerator1, Loop Delay=$User.EngineDelay, Loop Duration=1, UseLoopDelay=true)
   &SpawnBurst_Instantaneous(Spawn Count=1, Use Spawn Probability=true)
  ]
  particleSpawnScript [
   &FN_SetSpawnGroup(Number of Spawn Groups=1)
   &InitializeParticle(
    Color=&MakeCustomLinearColorFromBool(Bool=$User.bIsFrontEndPreview, True Color=<rgba 0.6599 1 0.845 1>, False Color=<rgba 1 1 1 1>)
    Lifetime=0.25
    Mesh Scale=&SelectVectorFromArray(Direct Array Index=$Particles.SpawnGroup, Vector Selection Array=DataInterface() {propertyValues {FloatData [<0.65 0.65 0.65>]}})
    Mesh Scale Mode=::NewEnumerator4
    Sprite Size Mode=::NewEnumerator3
    Uniform Sprite Size=220
    UsePositionOffset=false
    Write Material Random=true
   )
   &InitialMeshOrientation(Rotation=<0 0.25 0>, Use Orientation Vector=true, Use Rotation Vector=true) {module=InitialMeshOrientation001}
   &InitialMeshOrientation(
    Mesh Orientation Mode=::NewEnumerator1
    Rotation=&RandomRangeVector(Minimum=<-1 0 0>, Maximum=<1 0 0>)
    Use Rotation Vector=true
   )
   /SetVariables(
    Particles.Flame2Scale:Vector3f=&Multiply_Vector(A=$Particles.Scale, B=<1 1 0.9>)
    Particles.FlameCoreScale:Vector3f=&Multiply_Vector(A=$Particles.Scale, B=<0.65 0.6 0.6>)
   ) {isSetParameters=true, module=SetVariables_#2}
  ]
  particleUpdateScript [
   &ParticleState(Kill Particles When Lifetime Has Elapsed=false)
   &SkeletalMeshLocation(
    Apply Sampled Orientation to Mesh Particles=true
    Bone / Socket Index=&SelectIntFromArray(Direct Array Index=$Particles.SpawnGroup, Int Selection Array=DataInterface() {propertyValues {IntData [2]}})
    Inherit Velocity=false
    MaskSpawnGroup=false
    Mesh Sampling Type=::NewEnumerator5
    Skeletal Mesh=$User.SkeletalMesh
    Socket Sampling Mode=::NewEnumerator5
   )
   &SolveForcesAndVelocity(Clamp Velocity=false, Limit Acceleration=false)
   &ScaleColor(
    Scale Alpha=&FloatFromCurve(
     FloatCurve=DataInterface() {propertyValues {Curve {keys [<at 0 0 RCIM_Cubic>,<at 0.5 1 RCIM_Cubic>]}}}
     Scale Curve=&Multiply_Float(
      A=&Lerp_Float(
       A=0.15
       B=1
       Alpha=&FN_FrontEndFloatModifier(Gameplay=$System.FN_Integrator001.Value, Front End=$User.FE Intergrator Value)
      )
      B=&MakeCustomFloatFromBool(Bool=$User.bIsFullyDeployed, True Float=1, False Float=0)
     )
    )
    Scale RGB=&VectorFromFloat(Value=&FN_FrontEndFloatModifier(Gameplay=&FNTimeOfDayScalar(Night Scale=2, Morning Scale=1, Evening Scale=1), Front End=0.5))
    ScaleA=true
    ScaleRGB=true
   )
   /SetVariables(
    Particles.Flame2Color:LinearColor=&MakeLinearColorFromVectorAndFloat(
     "Vector (RGB)"=&VectorFromFloat(Value=&FN_FrontEndFloatModifier(Gameplay=2, Front End=1))
     "Float (Alpha)"=&Multiply_Float(
      A=&Multiply_Float(
       A=&Lerp_Float(A=0.5, B=0.7, Alpha=$System.FN_Integrator001.Value)
       B=&MakeCustomFloatFromBool(Bool=$User.bIsFullyDeployed, True Float=6, False Float=0)
      )
      B=&FN_FrontEndFloatModifier(Gameplay=1, Front End=0)
     )
    )
    Particles.FlameCoreColor:LinearColor=&MakeLinearColorFromVectorAndFloat(
     "Vector (RGB)"=&MakeVectorFromLinearColorRGB(Color=&MultiplyLinearColor(
       Linear Color=<rgba 10 4.7464 1.0514 1>
       Scale Factor=&MakeVector4_FromFloat(Float=&FN_FrontEndFloatModifier(Gameplay=1, Front End=1))
      ))
     "Float (Alpha)"=&Multiply_Float(
      A=&Lerp_Float(A=0.4, B=0.7, Alpha=$System.FN_Integrator001.Value)
      B=&MakeCustomFloatFromBool(Bool=$User.bIsFullyDeployed, True Float=1, False Float=0)
     )
    )
   ) {isSetParameters=true, module=SetVariables_#3}
   &ScaleMeshSize(Scale Factor=&MakeVector(
     X=1
     Y=1
     Z=&FloatFromCurve(FloatCurve=DataInterface() {propertyValues {Curve {keys [<at 0 0>,<at 0.85 1>]}}})
    ))
   &ScaleMeshSize(Scale Factor=&MakeVector(
     X=1
     Y=1
     Z=&Lerp_Float(
      A=0.8
      B=&SelectFloatFromArray(Direct Array Index=$Particles.SpawnGroup, Float Selection Array=DataInterface() {propertyValues {FloatData [1.8]}})
      Alpha=&FN_FrontEndFloatModifier(Gameplay=$System.FN_Integrator001.Value, Front End=$User.FE Intergrator Value)
     )
    )) {module=ScaleMeshSize001}
   /SetVariables(
    Particles.Flame2Scale:Vector3f=&Multiply_Vector(
     A=$Particles.Initial.Flame2Scale
     B=&MakeVector(
      X=&FloatFromCurve(FloatCurve=DataInterface() {propertyValues {Curve {keys [<at 0 0>,<at 1 1>]}}})
      Y=1
      Z=1
     )
    )
    Particles.FlameCoreScale:Vector3f=&Multiply_Vector(
     A=$Particles.Initial.FlameCoreScale
     B=&MakeVector(
      X=&FloatFromCurve(FloatCurve=DataInterface() {propertyValues {Curve {keys [<at 0 0>,<at 1 1>]}}})
      Y=1
      Z=1
     )
    )
   ) {isSetParameters=true, module=SetVariables_#4}
   /SetVariables(
    Particles.Flame2Scale:Vector3f=&Multiply_Vector(
     A=$Particles.Flame2Scale
     B=&MakeVector(
      X=1
      Y=1
      Z=&Lerp_Float(
       A=&SelectFloatFromArray(Direct Array Index=$Particles.SpawnGroup, Float Selection Array=DataInterface() {propertyValues {FloatData [1]}})
       B=&SelectFloatFromArray(Direct Array Index=$Particles.SpawnGroup, Float Selection Array=DataInterface() {propertyValues {FloatData [1.6]}})
       Alpha=&FN_FrontEndFloatModifier(Gameplay=$System.FN_Integrator001.Value, Front End=$User.FE Intergrator Value)
      )
     )
    )
    Particles.FlameCoreScale:Vector3f=&Multiply_Vector(
     A=$Particles.FlameCoreScale
     B=&MakeVector(
      X=1
      Y=1
      Z=&Lerp_Float(
       A=&SelectFloatFromArray(Direct Array Index=$Particles.SpawnGroup, Float Selection Array=DataInterface() {propertyValues {FloatData [1.25]}})
       B=&SelectFloatFromArray(Direct Array Index=$Particles.SpawnGroup, Float Selection Array=DataInterface() {propertyValues {FloatData [1.75]}})
       Alpha=&FN_FrontEndFloatModifier(Gameplay=$System.FN_Integrator001.Value, Front End=$User.FE Intergrator Value)
      )
     )
    )
   ) {isSetParameters=true, module=SetVariables_#5}
   &ScaleSpriteSize()
   &DynamicMaterialParameters(
    Index 0 Param 1=&Lerp_Float(
     A=0.3
     B=0.6
     Alpha=&FN_FrontEndFloatModifier(Gameplay=$System.FN_Integrator001.Value, Front End=$User.FE Intergrator Value)
    )
    Module.Param0WriteEnabled=true
    Module.Param1WriteEnabled=true
    Module.Param2WriteEnabled=true
    Module.Param3WriteEnabled=true
   )
  ]
  rendererClasses [&NiagaraMeshRendererProperties]
  renderers [
   {rendererClass=&NiagaraMeshRendererProperties, rendererIndex=0}
   {rendererClass=&NiagaraMeshRendererProperties, rendererIndex=1}
   {rendererClass=&NiagaraMeshRendererProperties, rendererIndex=2}
   {rendererClass=&NiagaraMeshRendererProperties, rendererIndex=3}
  ]
 }
 EngineFlames_Burst {
  emitterUpdateScript [
   &EmitterState(Life Cycle Mode=::NewEnumerator1, Loop Behavior=::NewEnumerator1, Loop Delay=$User.EngineDelay, Loop Duration=1, UseLoopDelay=true)
   &SpawnBurst_Instantaneous(Spawn Count=3, Use Spawn Probability=true)
  ]
  particleSpawnScript [
   &FN_SetSpawnGroup(Number of Spawn Groups=3)
   &InitializeParticle(
    Color=&MakeCustomLinearColorFromBool(Bool=$User.bIsFrontEndPreview, True Color=<rgba 0.6599 1 0.845 1>, False Color=<rgba 1 1 1 1>)
    Lifetime=1.2
    Mesh Scale=&SelectVectorFromArray(
     Direct Array Index=$Particles.SpawnGroup
     Vector Selection Array=DataInterface() {propertyValues {FloatData [<0.35 0.35 0.25>, <0.35 0.35 0.25>, <0.65 0.65 0.6>]}}
    )
    Mesh Scale Mode=::NewEnumerator4
    Sprite Size Mode=::NewEnumerator3
    Uniform Sprite Size=22
    UsePositionOffset=false
    Write Material Random=true
   )
   &InitialMeshOrientation(Rotation=<0 0.25 0>, Use Orientation Vector=true, Use Rotation Vector=true) {module=InitialMeshOrientation001}
   &InitialMeshOrientation(
    Mesh Orientation Mode=::NewEnumerator1
    Rotation=&RandomRangeVector(Minimum=<-1 0 0>, Maximum=<1 0 0>)
    Use Rotation Vector=true
   )
   /SetVariables(
    Particles.Flame2Scale:Vector3f=&Multiply_Vector(A=$Particles.Scale, B=<1 1 2>)
    Particles.FlameCoreScale:Vector3f=&Multiply_Vector(A=$Particles.Scale, B=<0.65 0.65 1>)
   ) {isSetParameters=true, module=SetVariables_#2}
  ]
  particleUpdateScript [
   &ParticleState()
   &SkeletalMeshLocation(
    Apply Sampled Orientation to Mesh Particles=true
    Bone / Socket Index=&SelectIntFromArray(Direct Array Index=$Particles.SpawnGroup, Int Selection Array=DataInterface() {propertyValues {IntData [0,1,2]}})
    Inherit Velocity=false
    MaskSpawnGroup=false
    Mesh Sampling Type=::NewEnumerator5
    Skeletal Mesh=$User.SkeletalMesh
    Socket Sampling Mode=::NewEnumerator5
   )
   &SolveForcesAndVelocity(Clamp Velocity=false, Limit Acceleration=false)
   &ScaleColor(
    Scale Alpha=&Multiply_Float(
     A=&FloatFromCurve(
      FloatCurve=DataInterface() {propertyValues {Curve {keys [
          <at 0 5 RCIM_Cubic>
          {arriveTangent=-4.3998, interpMode=RCIM_Cubic, leaveTangent=-4.3998, tangentMode=RCTM_User, time=0.25, value=2}
          {interpMode=RCIM_Cubic, tangentMode=RCTM_User, time=1, value=0}
         ]}}}
      Scale Curve=&MakeCustomFloatFromBool(Bool=$User.bIsFullyDeployed, True Float=1, False Float=0)
     )
     B=&FN_FrontEndFloatModifier(Gameplay=1, Front End=0.1)
    )
    Scale RGB=&Multiply_VectorByFloat(
     Vector=<1 1 1>
     Float=&FN_FrontEndFloatModifier(Gameplay=&FNTimeOfDayScalar(Night Scale=2, Morning Scale=1, Evening Scale=1), Front End=0.5)
    )
    ScaleA=true
    ScaleRGB=true
   )
   /SetVariables(Particles.EngineCoreColor:LinearColor=&MultiplyLinearColor(
     Linear Color=$Particles.Color
     Scale Factor=&MakeVector4_FromFloat(Float=&FloatFromCurve(
       FloatCurve=DataInterface() {propertyValues {Curve {keys [
           {arriveTangent=-2.5249, interpMode=RCIM_Cubic, leaveTangent=-2.5249, tangentMode=RCTM_User, time=0, value=0.9847}
           {arriveTangent=-0.0405, interpMode=RCIM_Cubic, leaveTangent=-0.0405, tangentMode=RCTM_User, time=1, value=0}
          ]}}}
       Scale Curve=20
      ))
    )) {isSetParameters=true}
   /SetVariables(
    Particles.Flame2Color:LinearColor=&MultiplyLinearColor(
     Linear Color=<rgba 1 1 1 1>
     Scale Factor=&MakeVector4_FromFloat(Float=&FloatFromCurve(
       FloatCurve=DataInterface() {propertyValues {Curve {keys [<at 0 5>,<at 0.2 1>,<at 1 0>]}}}
       Scale Curve=&FN_FrontEndFloatModifier(Gameplay=1, Front End=1)
      ))
    )
    Particles.FlameCoreColor:LinearColor=&MultiplyLinearColor(
     Linear Color=&MultiplyLinearColor(
      Linear Color=<rgba 10 6.5867 1.1458 1>
      Scale Factor=&MakeVector4_FromFloat(Float=&FN_FrontEndFloatModifier(Gameplay=1, Front End=1))
     )
     Scale Factor=&MakeVector4_FromFloat(Float=&FloatFromCurve(
       FloatCurve=DataInterface() {propertyValues {Curve {keys [<at 0 5>,<at 0.2 1>,<at 1 0>]}}}
       Scale Curve=&FN_FrontEndFloatModifier(Gameplay=1, Front End=1)
      ))
    )
   ) {isSetParameters=true, module=SetVariables_#3}
   /SetVariables(Particles.FlameCoreOrientation:Quat4f=&MultiplyQuaternion(Quaternion A=$Particles.MeshOrientation, Quaternion B=<0 1 0 1>)) {isSetParameters=true}
   &ScaleMeshSize(Scale Factor=&MakeVector(
     X=&FloatFromCurve(FloatCurve=DataInterface() {propertyValues {Curve {keys [<at 0 0>,<at 0.1 1.2>,<at 0.35 0.8>]}}})
     Y=&FloatFromCurve(FloatCurve=DataInterface() {propertyValues {Curve {keys [<at 0 0>,<at 0.1 1.2>,<at 0.35 0.8>]}}})
     Z=&FloatFromCurve(FloatCurve=DataInterface() {propertyValues {Curve {keys [
          {arriveTangent=4.4563, interpMode=RCIM_Cubic, leaveTangent=4.4563, tangentMode=RCTM_User, time=0, value=0}
          {interpMode=RCIM_Cubic, tangentMode=RCTM_User, time=0.2, value=5}
          <at 0.7 0>
         ]}}})
    ))
   /SetVariables(
    Particles.Flame2Scale:Vector3f=&Multiply_Vector(
     A=&Multiply_Vector(
      A=$Particles.Initial.Flame2Scale
      B=&MakeVector(
       X=1
       Y=1
       Z=&FloatFromCurve(FloatCurve=DataInterface() {propertyValues {Curve {keys [<at 0 0>,<at 0.1 1>,<at 1 0>]}}})
      )
     )
     B=<1 1 1>
    )
    Particles.FlameCoreScale:Vector3f=&Multiply_Vector(
     A=$Particles.Initial.FlameCoreScale
     B=&MakeVector(
      X=1
      Y=1
      Z=&FloatFromCurve(FloatCurve=DataInterface() {propertyValues {Curve {keys [<at 0 0>,<at 0.1 1>,<at 1 0>]}}})
     )
    )
   ) {isSetParameters=true, module=SetVariables_#4}
   &DynamicMaterialParameters(
    Index 0 Param 1=1
    Module.Param0WriteEnabled=true
    Module.Param1WriteEnabled=true
    Module.Param2WriteEnabled=true
    Module.Param3WriteEnabled=true
   )
  ]
  rendererClasses [&NiagaraMeshRendererProperties]
  renderers [
   {rendererClass=&NiagaraMeshRendererProperties, rendererIndex=0}
   {rendererClass=&NiagaraMeshRendererProperties, rendererIndex=1}
   {rendererClass=&NiagaraMeshRendererProperties, rendererIndex=2}
  ]
 }
 FlameMesh {
  emitterUpdateScript [
   &EmitterState(
    Inactive Response=::NewEnumerator1
    Life Cycle Mode=::NewEnumerator1
    Loop Behavior=::NewEnumerator1
    Loop Duration Mode=::NewEnumerator1
    Reset Age On Awaken=true
    Scalability Mode=::NewEnumerator1
    UseLoopDelay=false
   )
   &SpawnBurst_Instantaneous(Spawn Count=1, Use Spawn Probability=false)
  ]
  particleSpawnScript [
   /Niagara/Modules/Spawn/Initialization/InitializeParticle(
    Color=&MakeLinearColorFromVectorAndFloat(
     "Vector (RGB)"=&MakeVectorFromLinearColorRGB(Color=<rgba 300 220.283 125.755 1>)
     "Float (Alpha)"=&Multiply_Float(A=0.05, B=0.2)
    )
    Mesh Scale=<1.1 1 1>
    Module.Write Color=true
    Module.Write Lifetime=true
    Module.Write Mass=true
    Module.Write Material Random=true
    Module.Write Position=false
    Module.Write Scale=true
    Module.Write SpriteRotation=false
    Module.Write SpriteSize=false
    Position=&SimulationPosition()
    Sprite Size=<5 5>
   )
   &DynamicMaterialParameters(
    Index 0 Param 1=&RandomRangeFloat(Minimum=0, Maximum=1)
    Index 0 Param 2=1
    Index 0 Param 3=&FN_FrontEndFloatModifier(Gameplay=30, Front End=10)
    Module.Param0WriteEnabled=true
    Module.Param1WriteEnabled=true
    Module.Param2WriteEnabled=true
    Module.Param3WriteEnabled=true
   )
   &InitialMeshOrientation(
    Mesh Orientation Mode=::NewEnumerator2
    Rotation=&FN_FrontEndVector3Modifier(Gameplay=<0 0 0>, Front End=<0 0 0>)
    Use Rotation Vector=true
   )
  ]
  particleUpdateScript [
   &ParticleState(Kill Particles When Lifetime Has Elapsed=false, Let Infinitely Lived Particles Die When Emitter Deactivates=true)
   &InterpolateOverTime(
    Rate Of Change:NiagaraFloat=$User.Interpolate_Rate_of_Change
    Target Vector Value:NiagaraPosition=$Particles.Position
    Variable Type=::NewEnumerator2
   )
   /SetVariables(Particles.Interpolate Dynamic=&MakeFloat4FromVectorAndFloat("Vector (RGB)"=$Output.InterpolateOverTime.Moving Average, "Float (Alpha)"=0.1)) {isSetParameters=true, module=SetVariables_#6}
   &SolveForcesAndVelocity(Clamp Velocity=false, Limit Acceleration=false)
   &CameraOffset()
   &SkeletalMeshLocation(
    Apply Sampled Orientation to Mesh Particles=true
    Bone / Socket Index=3
    Inherit Velocity=false
    Interpolate Mesh Sampling=false
    MaskSpawnGroup=false
    Mesh Sampling Type=::NewEnumerator5
    Sampled Position Offset Coordinate Space:ENiagaraMeshTransforms=::NewEnumerator3
    Skeletal Mesh=$User.SkeletalMesh
    Socket Sampling Mode=::NewEnumerator5
    Use Sampled Mesh Attribute as Ribbon ID=false
   )
   &ScaleColor(
    Scale Alpha=&Multiply_Float(
     A=&FloatFromCurve(FloatCurve=DataInterface() {propertyValues {Curve {keys [<at 0 0>,<at 0.4 1>]}}}, CurveIndex=$System.Age)
     B=&FN_FrontEndFloatModifier(Gameplay=1, Front End=0.35)
    )
    Scale RGB=&VectorFromFloat(Value=&Multiply_Float(
      A=&FloatFromCurve(
       FloatCurve=DataInterface() {propertyValues {Curve {keys [
           {interpMode=RCIM_Cubic, tangentMode=RCTM_User, time=0.1, value=0.2}
           {interpMode=RCIM_Cubic, tangentMode=RCTM_User, time=0.25, value=3}
           <at 0.5 1>
          ]}}}
       CurveIndex=$System.Age
       Scale Curve=&FNTimeOfDayScalar(Night Scale=2, Morning Scale=1, Evening Scale=1)
      )
      B=&FN_FrontEndFloatModifier(Gameplay=0.7, Front End=1)
     ))
    ScaleA=true
    ScaleRGB=true
   )
   &ScaleMeshSize(Scale Factor=&FN_FrontEndVector3Modifier(Gameplay=<1 1 1>, Front End=<0.65 0.65 0.65>))
  ]
  rendererClasses [&NiagaraMeshRendererProperties]
  renderers [{rendererClass=&NiagaraMeshRendererProperties, rendererIndex=0}]
 }
 FlameMesh001 {
  emitterUpdateScript [
   &EmitterState(
    Inactive Response=::NewEnumerator1
    Life Cycle Mode=::NewEnumerator1
    Loop Behavior=::NewEnumerator1
    Loop Duration Mode=::NewEnumerator1
    Reset Age On Awaken=true
    Scalability Mode=::NewEnumerator1
    UseLoopDelay=false
   )
   &SpawnBurst_Instantaneous(Spawn Count=1, Use Spawn Probability=false)
  ]
  particleSpawnScript [
   &InitializeParticle(
    Color=<rgba 30 18.7626 2.8628 1>
    Mesh Scale=<0.76 1 1>
    Module.Write Color=true
    Module.Write Lifetime=true
    Module.Write Mass=true
    Module.Write Material Random=true
    Module.Write Position=false
    Module.Write Scale=true
    Module.Write SpriteRotation=false
    Module.Write SpriteSize=false
    Position=&SimulationPosition()
    Sprite Size=<5 5>
   )
   &DynamicMaterialParameters(
    Index 0 Param 1=&RandomRangeFloat(Minimum=0, Maximum=1)
    Index 0 Param 2=1
    Index 0 Param 3=&FN_FrontEndFloatModifier(Gameplay=20, Front End=12)
    Module.Param0WriteEnabled=true
    Module.Param1WriteEnabled=true
    Module.Param2WriteEnabled=true
    Module.Param3WriteEnabled=true
   )
   &InitialMeshOrientation(
    Mesh Orientation Mode=::NewEnumerator2
    Rotation=&FN_FrontEndVector3Modifier(Gameplay=<0 0 0>, Front End=<0 0 0>)
    Use Rotation Vector=true
   )
  ]
  particleUpdateScript [
   &ParticleState(Kill Particles When Lifetime Has Elapsed=false, Let Infinitely Lived Particles Die When Emitter Deactivates=true)
   &InterpolateOverTime(
    Rate Of Change:NiagaraFloat=$User.Interpolate_Rate_of_Change
    Target Vector Value:NiagaraPosition=$Particles.Position
    Variable Type=::NewEnumerator2
   )
   /SetVariables(Particles.Interpolate Dynamic=&MakeFloat4FromVectorAndFloat("Vector (RGB)"=$Output.InterpolateOverTime.Moving Average, "Float (Alpha)"=0)) {isSetParameters=true, module=SetVariables_#6}
   &SolveForcesAndVelocity(Clamp Velocity=false, Limit Acceleration=false)
   &CameraOffset()
   &SkeletalMeshLocation(
    Apply Sampled Orientation to Mesh Particles=true
    Bone / Socket Index=3
    Inherit Velocity=false
    Interpolate Mesh Sampling=false
    MaskSpawnGroup=false
    Mesh Sampling Type=::NewEnumerator5
    Sampled Position Offset Coordinate Space:ENiagaraMeshTransforms=::NewEnumerator3
    Skeletal Mesh=$User.SkeletalMesh
    Socket Sampling Mode=::NewEnumerator5
    Use Sampled Mesh Attribute as Ribbon ID=false
   )
   &ScaleColor(
    Scale Alpha=&FloatFromCurve(
     FloatCurve=DataInterface() {propertyValues {Curve {keys [<at 0 0>,<at 0.4 1>]}}}
     CurveIndex=$System.Age
     Scale Curve=&FN_FrontEndFloatModifier(
      Gameplay=&FNTimeOfDayScalar(Night Scale=0.2, Morning Scale=0.5, Day Scale=0.5, Evening Scale=0.5)
      Front End=0.5
     )
    )
    Scale RGB=&VectorFromFloat(Value=&FloatFromCurve(
      FloatCurve=DataInterface() {propertyValues {Curve {keys [
          {interpMode=RCIM_Cubic, tangentMode=RCTM_User, time=0.1, value=0.2}
          {interpMode=RCIM_Cubic, tangentMode=RCTM_User, time=0.25, value=3}
          <at 0.5 1>
         ]}}}
      CurveIndex=$System.Age
      Scale Curve=&FNTimeOfDayScalar(Night Scale=3, Morning Scale=1, Evening Scale=1)
     ))
    ScaleA=true
    ScaleRGB=true
   )
   &ScaleMeshSize(Scale Factor=&FN_FrontEndVector3Modifier(Gameplay=<1 1 1>, Front End=<0.65 0.65 0.65>))
   &FN_Light_Attributes(
    Color=<rgba 20 9.4878 1.9792 1>
    Create Color Attribute=true
    Create Exponent Attribute=true
    Create Position Attribute=true
    Create Radius Attribute=true
    Exponent=2
    Position=&AddVectorToPosition(Position=$Particles.Position, Vector=<0 0 30>)
    Radius=70
   )
  ]
  rendererClasses [&NiagaraMeshRendererProperties, &NiagaraLightRendererProperties]
  renderers [
   {rendererClass=&NiagaraMeshRendererProperties, rendererIndex=0}
   {rendererClass=&NiagaraLightRendererProperties, rendererIndex=1}
  ]
 }
 Flare {
  emitterUpdateScript [
   &EmitterState(
    Life Cycle Mode=::NewEnumerator1
    Loop Behavior=::NewEnumerator1
    Loop Delay=$User.EngineDelay
    Loop Duration Mode=::NewEnumerator1
    UseLoopDelay=true
   )
   &SpawnBurst_Instantaneous(Spawn Count=3, Use Spawn Probability=false, UseLoopCountLimit=false)
  ]
  particleSpawnScript [
   &FN_SetSpawnGroup(Number of Spawn Groups=3)
   /Niagara/Modules/Spawn/Initialization/V2/InitializeParticle(
    Color=<rgba 2 0.9466 0.1941 1>
    Lifetime=0.25
    Position Offset=<0 0 50>
    Sprite Size=&Multiply_Vector2D_ByFloat(
     Float=&SelectFloatFromArray(Direct Array Index=$Particles.SpawnGroup, Float Selection Array=DataInterface() {propertyValues {FloatData [1.25,1.25,3]}})
     Vector2D=<80 50>
    )
    Sprite Size Mode=::NewEnumerator4
    UsePositionOffset=false
    Write Material Random=true
   )
  ]
  particleUpdateScript [
   &ParticleState(Kill Particles When Lifetime Has Elapsed=false, Let Infinitely Lived Particles Die When Emitter Deactivates=true)
   &SkeletalMeshLocation(
    Apply Sampled Orientation to Mesh Particles=true
    Bone / Socket Index=&SelectIntFromArray(Direct Array Index=$Particles.SpawnGroup, Int Selection Array=DataInterface() {propertyValues {IntData [0,1,2]}})
    Inherit Velocity=false
    MaskSpawnGroup=false
    Mesh Sampling Type=::NewEnumerator5
    Sampled Position Offset=<4 0 0>
    Sampled Position Offset Coordinate Space:ENiagaraMeshTransforms=::NewEnumerator3
    Skeletal Mesh=$User.SkeletalMesh
    Socket Sampling Mode=::NewEnumerator5
   )
   &ScaleSpriteSize(
    Scale Sprite Size Mode=::NewEnumerator0
    Uniform Scale Factor=&Multiply_Float(
     A=&FloatFromCurve(FloatCurve=DataInterface() {propertyValues {Curve {keys [<at 0 0>,<at 1 1>]}}})
     B=&Lerp_Float(
      A=0.6
      B=1
      Alpha=&FN_FrontEndFloatModifier(Gameplay=$System.FN_Integrator001.Value, Front End=$User.FE Intergrator Value)
     )
    )
   )
   &ScaleColor(
    Scale Alpha=&FNTimeOfDayScalar(Night Scale=1, Morning Scale=1, Evening Scale=1)
    Scale RGB=&Multiply_VectorByFloat(Vector=<1 1 1>, Float=&FN_FrontEndFloatModifier(Gameplay=1.5, Front End=0.6))
    ScaleA=true
    ScaleRGB=true
   )
   &CameraOffset(Camera Offset Amount=20)
   &DynamicMaterialParameters(
    Index 0 Param 1=0.4
    Index 0 Param 2=0.5
    Module.Param0WriteEnabled=true
    Module.Param1WriteEnabled=true
    Module.Param2WriteEnabled=true
    Module.Param3WriteEnabled=true
   )
   &FN_Light_v2(
    Light Exponent=2
    Light Radius:NiagaraFloat=&SelectFloatFromArray(Direct Array Index=$Particles.SpawnGroup, Float Selection Array=DataInterface() {propertyValues {FloatData [40,40,80]}})
    LightEnabled:NiagaraBool=$User.bIsFullyDeployed
    Position Offset:Vector3f=<0 -25 0>
    Scale Color=1
   )
   /SetVariables(Particles.Light Color:LinearColor=&MultiplyLinearColor(
     Linear Color=&MultiplyLinearColor(
      Linear Color=<rgba 100 39.8034 10.3803 1>
      Scale Factor=&MakeVector4_FromFloat(Float=&FN_FrontEndFloatModifier(Gameplay=1, Front End=0))
     )
     Scale Factor=&MakeVector4_FromFloat(Float=&Lerp_Float(
       A=0.5
       B=1
       Alpha=&FN_FrontEndFloatModifier(Gameplay=$System.FN_Integrator001.Value, Front End=$User.FE Intergrator Value)
      ))
    )) {isSetParameters=true, module=SetVariables_#7}
  ]
  rendererClasses [&NiagaraSpriteRendererProperties, &NiagaraLightRendererProperties]
  renderers [
   {rendererClass=&NiagaraSpriteRendererProperties}
   {rendererClass=&NiagaraLightRendererProperties, rendererIndex=1}
  ]
 }
 Flare_Burst {
  emitterUpdateScript [
   &EmitterState(Life Cycle Mode=::NewEnumerator1, Loop Behavior=::NewEnumerator1, Loop Delay=$User.EngineDelay, Loop Duration=1, UseLoopDelay=true)
   &SpawnBurst_Instantaneous(Spawn Count=3, Use Spawn Probability=false, UseLoopCountLimit=false)
  ]
  particleSpawnScript [
   &FN_SetSpawnGroup(Number of Spawn Groups=3)
   &InitializeParticle(
    Color=<rgba 5 1.6667 0 1>
    Lifetime=0.7
    Position Offset=<0 0 50>
    Sprite Size=&Multiply_Vector2D_ByFloat(
     Float=&SelectFloatFromArray(Direct Array Index=$Particles.SpawnGroup, Float Selection Array=DataInterface() {propertyValues {FloatData [2.3,2.3,5]}})
     Vector2D=<80 50>
    )
    Sprite Size Mode=::NewEnumerator4
    UsePositionOffset=false
    Write Material Random=true
   )
  ]
  particleUpdateScript [
   &ParticleState()
   &SkeletalMeshLocation(
    Apply Sampled Orientation to Mesh Particles=true
    Bone / Socket Index=&SelectIntFromArray(Direct Array Index=$Particles.SpawnGroup, Int Selection Array=DataInterface() {propertyValues {IntData [0,1,2]}})
    Inherit Velocity=false
    MaskSpawnGroup=false
    Mesh Sampling Type=::NewEnumerator5
    Sampled Position Offset=<4 0 0>
    Sampled Position Offset Coordinate Space:ENiagaraMeshTransforms=::NewEnumerator3
    Skeletal Mesh=$User.SkeletalMesh
    Socket Sampling Mode=::NewEnumerator5
   )
   &ScaleSpriteSize(
    Scale Sprite Size Mode=::NewEnumerator0
    Uniform Scale Factor=&Multiply_Float(
     A=&FloatFromCurve(FloatCurve=DataInterface() {propertyValues {Curve {keys [
          {arriveTangent=2.2885, interpMode=RCIM_Cubic, leaveTangent=2.2885, tangentMode=RCTM_User, time=0, value=0}
          {arriveTangent=-2.3026, interpMode=RCIM_Cubic, leaveTangent=-2.3026, tangentMode=RCTM_User, time=0.2, value=1.7}
          <at 1 0>
         ]}}})
     B=1
    )
   )
   &ScaleColor(
    Scale Alpha=&FloatFromCurve(
     FloatCurve=DataInterface() {propertyValues {Curve {keys [
         {interpMode=RCIM_Cubic, tangentMode=RCTM_User, time=0.4, value=1}
         {arriveTangent=-1.5698, interpMode=RCIM_Cubic, leaveTangent=-1.5698, tangentMode=RCTM_User, time=1, value=0}
        ]}}}
     Scale Curve=&FN_FrontEndFloatModifier(Gameplay=1, Front End=0.2)
    )
    ScaleA=true
    ScaleRGB=true
   )
   &FN_Light_v2(
    Light Exponent=2
    Light Radius:NiagaraFloat=&SelectFloatFromArray(Direct Array Index=$Particles.SpawnGroup, Float Selection Array=DataInterface() {propertyValues {FloatData [40,40,80]}})
    LightEnabled:NiagaraBool=$User.bIsFullyDeployed
    Position Offset:Vector3f=<0 -25 0>
    Scale Color=1
   )
   /SetVariables(Particles.Light Color:LinearColor=&MultiplyLinearColor(
     Linear Color=<rgba 100 47.5 10 1>
     Scale Factor=&MakeVector4_FromFloat(Float=&FloatFromCurve(
       FloatCurve=DataInterface() {propertyValues {Curve {keys [
           {interpMode=RCIM_Cubic, tangentMode=RCTM_User, time=0.4, value=1.5}
           {arriveTangent=-3.3732, interpMode=RCIM_Cubic, leaveTangent=-3.3732, tangentMode=RCTM_User, time=1, value=0}
          ]}}}
       Scale Curve=&FN_FrontEndFloatModifier(Gameplay=1, Front End=0)
      ))
    )) {isSetParameters=true, module=SetVariables_#7}
   &CameraOffset(Camera Offset Amount=20)
   &DynamicMaterialParameters(
    Index 0 Param 1=0.4
    Index 0 Param 2=0.5
    Module.Param0WriteEnabled=true
    Module.Param1WriteEnabled=true
    Module.Param2WriteEnabled=true
    Module.Param3WriteEnabled=true
   )
  ]
  rendererClasses [&NiagaraSpriteRendererProperties]
  renderers [{rendererClass=&NiagaraSpriteRendererProperties}]
 }
 Hair01 {
  emitterUpdateScript [
   &EmitterState(Life Cycle Mode=::NewEnumerator1, Loop Behavior=::NewEnumerator1, Loop Duration Mode=::NewEnumerator1, UseLoopDelay=false)
   &SpawnBurst_Instantaneous(Use Spawn Probability=false, UseLoopCountLimit=false)
  ]
  particleSpawnScript [
   &InitializeParticle(
    Color=&MakeCustomLinearColorFromBool(Bool=$User.bIsFrontEndPreview, True Color=<rgba 0.6851 1 0.8204 2>, False Color=<rgba 1 1 1 2>)
    Mesh Scale Mode=::NewEnumerator3
    UsePositionOffset=false
    Write Material Random=true
   )
   &InitialMeshOrientation(
    Mesh Orientation Mode=::NewEnumerator2
    Rotation=&FN_FrontEndVector3Modifier(Gameplay=<0 0 0>, Front End=<0 0 0>)
    Use Rotation Vector=true
   )
  ]
  particleUpdateScript [
   &ParticleState(Kill Particles When Lifetime Has Elapsed=false, Let Infinitely Lived Particles Die When Emitter Deactivates=true)
   &InterpolateOverTime(
    Rate Of Change:NiagaraFloat=$User.Interpolate_Rate_of_Change
    Target Vector Value:NiagaraPosition=$Particles.Position
    Variable Type=::NewEnumerator2
   )
   /SetVariables(Particles.Interpolate Dynamic=&MakeFloat4FromVectorAndFloat("Vector (RGB)"=$Output.InterpolateOverTime.Moving Average, "Float (Alpha)"=0.2)) {isSetParameters=true}
   &SkeletalMeshLocation(
    Bone / Socket Index=3
    MaskSpawnGroup=false
    Mesh Sampling Type=::NewEnumerator5
    Skeletal Mesh=$User.SkeletalMesh
    Socket Sampling Mode=::NewEnumerator5
   )
   &DynamicMaterialParameters(
    Index 0 Param 4=&Multiply_Float(
     A=&FN_FrontEndFloatModifier(Gameplay=1, Front End=0.7)
     B=&FloatFromCurve(
      FloatCurve=DataInterface() {propertyValues {Curve {keys [
          {interpMode=RCIM_Cubic, leaveTangentWeight=0.3333, tangentMode=RCTM_Break, time=-1, value=1}
          {arriveTangent=-0.7044, interpMode=RCIM_Cubic, leaveTangent=0.6442, tangentMode=RCTM_Break, time=0, value=0.8}
          {arriveTangentWeight=0.3333, interpMode=RCIM_Cubic, tangentMode=RCTM_Break, time=1, value=1}
         ]}}}
      CurveIndex=&DotProduct(
       Vector A=&ReturnCameraVector(Camera Vectors=::NewEnumerator1, Coordinate Space=::World)
       Vector B=&RotateVectorByQuaternion(Vector To Rotate=<1 0 0>, Quaternion=$StackContext.SkeletalMeshLocation.SampledRotation)
      )
     )
    )
    Index 2 Param 1=&FloatFromCurve(FloatCurve=DataInterface() {propertyValues {Curve {keys [<at 0 0>,<at 0.2 10>,<at 0.5 1>]}}})
    Index 2 Param 2=&FloatFromCurve(
     FloatCurve=DataInterface() {propertyValues {Curve {keys [
         {interpMode=RCIM_Cubic, leaveTangent=-2.2365, tangentMode=RCTM_Break, time=0, value=0}
         {arriveTangent=-3.655, interpMode=RCIM_Cubic, tangentMode=RCTM_Break, time=0.55, value=-1.3}
         <at 1 -1.4>
        ]}}}
     CurveIndex=$System.Age
    )
    Index 2 Param 3=&FloatFromCurve(
     FloatCurve=DataInterface() {propertyValues {Curve {keys [{interpMode=RCIM_Cubic, tangentMode=RCTM_User, time=0.1, value=8}, {interpMode=RCIM_Cubic, tangentMode=RCTM_User, time=0.6, value=0}]}}}
     CurveIndex=$System.Age
    )
    Index 2 Param 4=&FloatFromCurve(
     FloatCurve=DataInterface() {propertyValues {Curve {keys [
         {arriveTangent=3.1135, interpMode=RCIM_Cubic, leaveTangent=3.1135, tangentMode=RCTM_User, time=0, value=0}
         {arriveTangent=1.1153, interpMode=RCIM_Cubic, leaveTangent=1.1153, tangentMode=RCTM_User, time=0.5, value=1}
        ]}}}
     CurveIndex=$System.Age
    )
    Index 3 Param 1=1
    Index 3 Param 4=$Particles.Age
    Module.Param0WriteEnabled=true
    Module.Param0WriteEnabled2=true
    Module.Param0WriteEnabled3=true
    Module.Param1WriteEnabled=true
    Module.Param1WriteEnabled2=true
    Module.Param1WriteEnabled3=true
    Module.Param2WriteEnabled=true
    Module.Param2WriteEnabled2=true
    Module.Param2WriteEnabled3=true
    Module.Param3WriteEnabled=true
    Module.Param3WriteEnabled2=true
    Module.Param3WriteEnabled3=true
    Write Parameter Index 2=true
    Write Parameter Index 3=true
   )
   &ScaleMeshSize(Scale Factor=&FN_FrontEndVector3Modifier(Gameplay=<1 1 1>, Front End=<0.7 0.7 0.7>))
   &ScaleColor(Scale Alpha=0.3, Scale RGB=&Multiply_VectorByFloat(Vector=<1 1 1>, Float=0.5), ScaleA=true, ScaleRGB=true)
  ]
  rendererClasses [&NiagaraMeshRendererProperties]
  renderers [
   {rendererClass=&NiagaraMeshRendererProperties, rendererIndex=0}
   {rendererClass=&NiagaraMeshRendererProperties, rendererIndex=1}
  ]
 }
 Hair_Flame {
  emitterUpdateScript [
   &EmitterState(Life Cycle Mode=::NewEnumerator1, Loop Behavior=::NewEnumerator1, Loop Duration Mode=::NewEnumerator1, UseLoopDelay=false)
   &SpawnBurst_Instantaneous(Use Spawn Probability=false, UseLoopCountLimit=false)
  ]
  particleSpawnScript [
   &InitializeParticle(
    Color=&MakeCustomLinearColorFromBool(Bool=$User.bIsFrontEndPreview, True Color=<rgba 1 1 1 1>, False Color=<rgba 1 0.8378 0.7977 1>)
    Mesh Scale Mode=::NewEnumerator3
    UsePositionOffset=false
    Write Material Random=true
   )
   &InitialMeshOrientation(
    Mesh Orientation Mode=::NewEnumerator2
    Rotation=&FN_FrontEndVector3Modifier(Gameplay=<0 0 0>, Front End=<0 0 0>)
    Use Rotation Vector=true
   )
  ]
  particleUpdateScript [
   &ParticleState(Kill Particles When Lifetime Has Elapsed=false, Let Infinitely Lived Particles Die When Emitter Deactivates=true)
   &SkeletalMeshLocation(
    Bone / Socket Index=3
    MaskSpawnGroup=false
    Mesh Sampling Type=::NewEnumerator5
    Skeletal Mesh=$User.SkeletalMesh
    Socket Sampling Mode=::NewEnumerator5
   )
   &InterpolateOverTime(
    Rate Of Change:NiagaraFloat=$User.Interpolate_Rate_of_Change
    Target Vector Value:NiagaraPosition=$Particles.Position
    Variable Type=::NewEnumerator2
   )
   /SetVariables(Particles.Interpolate Dynamic=&MakeFloat4FromVectorAndFloat("Vector (RGB)"=$Output.InterpolateOverTime.Moving Average, "Float (Alpha)"=1)) {isSetParameters=true}
   &DynamicMaterialParameters(
    Index 0 Param 4=&Multiply_Float(
     A=&FN_FrontEndFloatModifier(Gameplay=1, Front End=0.6)
     B=&FloatFromCurve(
      FloatCurve=DataInterface() {propertyValues {Curve {keys [
          {interpMode=RCIM_Cubic, leaveTangentWeight=0.3333, tangentMode=RCTM_Break, time=-1, value=1}
          {arriveTangent=-0.3734, interpMode=RCIM_Cubic, leaveTangent=0.3743, tangentMode=RCTM_Break, time=0, value=0.75}
          {arriveTangentWeight=0.3333, interpMode=RCIM_Cubic, tangentMode=RCTM_Break, time=1, value=1}
         ]}}}
      CurveIndex=&DotProduct(
       Vector A=&ReturnCameraVector(Camera Vectors=::NewEnumerator1, Coordinate Space=::World)
       Vector B=&RotateVectorByQuaternion(Vector To Rotate=<1 0 0>, Quaternion=$StackContext.SkeletalMeshLocation.SampledRotation)
      )
      Scale Curve=1.5
     )
    )
    Index 3 Param 1=&FloatFromCurve(
     FloatCurve=DataInterface() {propertyValues {Curve {keys [
         {arriveTangent=0.4948, interpMode=RCIM_Cubic, leaveTangent=0.4948, tangentMode=RCTM_User, time=0, value=0}
         {arriveTangent=0.5786, interpMode=RCIM_Cubic, leaveTangent=0.5786, tangentMode=RCTM_User, time=0.8, value=1}
        ]}}}
     CurveIndex=$System.Age
    )
    Index 3 Param 2=&Multiply_Float(A=$Particles.Age, B=&RandomRangeFloat(Minimum=1, Maximum=1.5))
    Module.Param0WriteEnabled=true
    Module.Param0WriteEnabled3=true
    Module.Param1WriteEnabled=true
    Module.Param1WriteEnabled3=true
    Module.Param2WriteEnabled=true
    Module.Param2WriteEnabled3=true
    Module.Param3WriteEnabled=true
    Module.Param3WriteEnabled3=true
    Write Parameter Index 3=true
   )
   &ScaleColor(
    Scale Alpha=&FN_FrontEndFloatModifier(Gameplay=0.7, Front End=1)
    Scale RGB=&VectorFromFloat(Value=&Multiply_Float(
      A=&FloatFromCurve(
       FloatCurve=DataInterface() {propertyValues {Curve {keys [<at 0 1>,<at 0.25 3>,<at 0.5 1>]}}}
       Scale Curve=&FN_FrontEndFloatModifier(Gameplay=1, Front End=0.6)
      )
      B=&FNTimeOfDayScalar(Night Scale=1.5, Morning Scale=1, Evening Scale=1)
     ))
    ScaleA=true
    ScaleRGB=true
   )
   /SetVariables(
    Particles.New Color:LinearColor=&MultiplyLinearColor(
     Linear Color=$Particles.Color
     Scale Factor=&MakeFloat4FromVectorAndFloat(
      "Vector (RGB)"=&VectorFromFloat(Value=&FN_FrontEndFloatModifier(
        Gameplay=&Multiply_Float(A=1.5, B=&FNTimeOfDayScalar(Night Scale=1.5, Morning Scale=1, Evening Scale=1))
        Front End=1.1
       ))
      "Float (Alpha)"=&FN_FrontEndFloatModifier(Gameplay=1, Front End=1)
     )
    )
    Particles.New Vector 4:Vector4f=&MakeVector4(
     X=0
     Y=0
     Z=0
     W=&FloatFromCurve(
      FloatCurve=DataInterface() {propertyValues {Curve {keys [
          {interpMode=RCIM_Cubic, leaveTangentWeight=0.3333, tangentMode=RCTM_Break, time=-1, value=1}
          {arriveTangent=-0.7044, interpMode=RCIM_Cubic, leaveTangent=0.6442, tangentMode=RCTM_Break, time=0, value=0.5}
          {arriveTangentWeight=0.3333, interpMode=RCIM_Cubic, tangentMode=RCTM_Break, time=1, value=1}
         ]}}}
      CurveIndex=&DotProduct(
       Vector A=&ReturnCameraVector(Camera Vectors=::NewEnumerator1, Coordinate Space=::World)
       Vector B=&RotateVectorByQuaternion(Vector To Rotate=<1 0 0>, Quaternion=$StackContext.SkeletalMeshLocation.SampledRotation)
      )
     )
    )
   ) {isSetParameters=true}
   &ScaleMeshSize(Scale Factor=&FN_FrontEndVector3Modifier(Gameplay=<1 1 1>, Front End=<0.65 0.65 0.65>))
  ]
  rendererClasses [&NiagaraMeshRendererProperties]
  renderers [
   {rendererClass=&NiagaraMeshRendererProperties, rendererIndex=0}
   {rendererClass=&NiagaraMeshRendererProperties, rendererIndex=1}
  ]
 }
 Hair_Thin {
  emitterUpdateScript [
   &EmitterState(Life Cycle Mode=::NewEnumerator1, Loop Behavior=::NewEnumerator1, Loop Duration Mode=::NewEnumerator1, UseLoopDelay=false)
   &SpawnBurst_Instantaneous(Use Spawn Probability=false, UseLoopCountLimit=false)
  ]
  particleSpawnScript [
   &InitializeParticle(Color=<rgba 3 3 3 0.5>, Mesh Scale Mode=::NewEnumerator3, UsePositionOffset=false, Write Material Random=true)
   &InitialMeshOrientation(
    Mesh Orientation Mode=::NewEnumerator2
    Rotation=&FN_FrontEndVector3Modifier(Gameplay=<0 0 0>, Front End=<0 0 0>)
    Use Rotation Vector=true
   )
  ]
  particleUpdateScript [
   &ParticleState(Kill Particles When Lifetime Has Elapsed=false, Let Infinitely Lived Particles Die When Emitter Deactivates=true)
   &InterpolateOverTime(
    Rate Of Change:NiagaraFloat=$User.Interpolate_Rate_of_Change
    Target Vector Value:NiagaraPosition=$Particles.Position
    Variable Type=::NewEnumerator2
   )
   /SetVariables(Particles.Interpolate Dynamic=&MakeFloat4FromVectorAndFloat("Vector (RGB)"=$Output.InterpolateOverTime.Moving Average, "Float (Alpha)"=0)) {isSetParameters=true}
   &SkeletalMeshLocation(
    Bone / Socket Index=3
    MaskSpawnGroup=false
    Mesh Sampling Type=::NewEnumerator5
    Skeletal Mesh=$User.SkeletalMesh
    Socket Sampling Mode=::NewEnumerator5
   )
   &DynamicMaterialParameters(
    Index 0 Param 4=&FN_FrontEndFloatModifier(Gameplay=1, Front End=0.7)
    Index 2 Param 1=&FloatFromCurve(FloatCurve=DataInterface() {propertyValues {Curve {keys [<at 0 0>,<at 0.2 10>,<at 0.5 1>]}}})
    Index 2 Param 2=&FloatFromCurve(
     FloatCurve=DataInterface() {propertyValues {Curve {keys [
         {interpMode=RCIM_Cubic, leaveTangent=-5.3954, tangentMode=RCTM_Break, time=0, value=0}
         {arriveTangent=-3.655, interpMode=RCIM_Cubic, tangentMode=RCTM_Break, time=0.4, value=-1.3}
         <at 1 -1.4>
        ]}}}
     CurveIndex=$System.Age
    )
    Index 2 Param 3=&FloatFromCurve(
     FloatCurve=DataInterface() {propertyValues {Curve {keys [{interpMode=RCIM_Cubic, tangentMode=RCTM_User, time=0.1, value=0.5}, {interpMode=RCIM_Cubic, tangentMode=RCTM_User, time=0.4, value=0}]}}}
     CurveIndex=$System.Age
    )
    Index 2 Param 4=&FloatFromCurve(
     FloatCurve=DataInterface() {propertyValues {Curve {keys [
         {arriveTangent=5.5002, interpMode=RCIM_Cubic, leaveTangent=5.5002, tangentMode=RCTM_User, time=0.1, value=0}
         {arriveTangent=1.1153, interpMode=RCIM_Cubic, leaveTangent=1.1153, tangentMode=RCTM_User, time=0.3, value=0.9}
        ]}}}
     CurveIndex=$System.Age
    )
    Index 3 Param 1=1
    Index 3 Param 4=$Particles.Age
    Module.Param0WriteEnabled=true
    Module.Param0WriteEnabled2=true
    Module.Param0WriteEnabled3=true
    Module.Param1WriteEnabled=true
    Module.Param1WriteEnabled2=true
    Module.Param1WriteEnabled3=true
    Module.Param2WriteEnabled=true
    Module.Param2WriteEnabled2=true
    Module.Param2WriteEnabled3=true
    Module.Param3WriteEnabled=true
    Module.Param3WriteEnabled2=true
    Module.Param3WriteEnabled3=true
    Write Parameter Index 2=true
    Write Parameter Index 3=true
   )
   &ScaleMeshSize(Scale Factor=&FN_FrontEndVector3Modifier(Gameplay=<1 1 1>, Front End=<0.7 0.7 0.7>))
   &ScaleColor(
    Scale Alpha=&FN_FrontEndFloatModifier(Gameplay=1, Front End=0.5)
    Scale RGB=&VectorFromFloat(Value=1)
    ScaleA=true
    ScaleRGB=true
   )
  ]
  rendererClasses [&NiagaraMeshRendererProperties]
  renderers [{rendererClass=&NiagaraMeshRendererProperties, rendererIndex=0}]
 }
 Spark_Burst {
  emitterUpdateScript [
   &EmitterState(Life Cycle Mode=::NewEnumerator1, Loop Behavior=::NewEnumerator1, Loop Delay=$User.EngineDelay, Loop Duration=1, UseLoopDelay=true)
   &SpawnRate(
    SpawnRate=&FloatFromCurve(
     FloatCurve=DataInterface() {propertyValues {Curve {keys [{interpMode=RCIM_Cubic, tangentMode=RCTM_User, time=0, value=1}, {interpMode=RCIM_Cubic, tangentMode=RCTM_User, time=1, value=0}]}}}
     CurveIndex=$Emitter.NormalizedLoopAge
     Scale Curve=20
    )
    Use Spawn Probability=true
   )
   &SpawnBurst_Instantaneous(Spawn Count=15, Use Spawn Probability=false, UseLoopCountLimit=false)
  ]
  particleSpawnScript [
   &FN_SetSpawnGroup(Number of Spawn Groups=2)
   &InitializeParticle(
    Color Maximum=<rgba 10 4.6167 0.5 1>
    Color Minimum=<rgba 10 3.8911 0.6366 1>
    Color Mode=::NewEnumerator2
    Lifetime Max=0.42
    Lifetime Min=0.15
    Lifetime Mode=::NewEnumerator1
    Sprite Size Max=<0.8 5.5>
    Sprite Size Min=<0.6 4>
    Sprite Size Mode=::NewEnumerator2
    UsePositionOffset=false
    Write Material Random=true
   )
   &SkeletalMeshLocation(
    Apply Sampled Orientation to Mesh Particles=true
    Bone / Socket Index=&SelectIntFromArray(Direct Array Index=$Particles.SpawnGroup, Int Selection Array=DataInterface() {propertyValues {IntData [0,1]}})
    Inherit Velocity=false
    Interpolate Mesh Sampling=false
    MaskSpawnGroup=false
    Mesh Sampling Type=::NewEnumerator5
    Sampled Position Offset Coordinate Space:ENiagaraMeshTransforms=::NewEnumerator3
    Skeletal Mesh=$User.SkeletalMesh
    Socket Sampling Mode=::NewEnumerator5
   )
   &ShapeLocation(MaskSpawnGroup=false, Override Seed=false, Sphere Radius=10, UseOwnerScale=true)
   &AddVelocity(
    Cone Angle=50
    Cone Axis=&FN_FrontEndVector3Modifier(
     Gameplay=&SelectVectorFromArray(
      Direct Array Index=$Particles.SpawnGroup
      Vector Selection Array=DataInterface() {propertyValues {FloatData [<-0.3 -1 -0.4>, <0.3 -1 -0.4>]}}
     )
     Front End=&SelectVectorFromArray(Direct Array Index=$Particles.SpawnGroup, Vector Selection Array=DataInterface() {propertyValues {FloatData [<0.3 1 -0.4>, <-0.3 1 -0.4>]}})
    )
    Module.UseConeFalloff=false
    Module.UseVelDistribution=false
    Override Seed=false
    Velocity Mode=::NewEnumerator2
    Velocity Speed=&RandomRangeFloat(Minimum=400, Maximum=800)
   )
   /SetVariables(
    Particles.GlowColor=<rgba 0.2 0.05 0 1>
    Particles.GlowSize=&Multiply_Vector2D(A=$Particles.SpriteSize, B=<10 4>)
   ) {isSetParameters=true, module=SetVariables_#8}
  ]
  particleUpdateScript [
   &ParticleState()
   &CurlNoiseForce(Noise Frequency=200, Noise Strength=4000, Pan Noise=false)
   &Drag(Drag=&RandomRangeFloat(Minimum=5, Maximum=6), Use Linear Drag=true, Use Rotational Drag=false)
   &SolveForcesAndVelocity(Clamp Velocity=false, Limit Acceleration=false)
   &ScaleSpriteSize(
    Non-Uniform Curve Sprite Scale=DataInterface() {propertyValues {
      XCurve {keys [
        <at 0 1.5>
        {arriveTangent=-1.0124, interpMode=RCIM_Cubic, leaveTangent=-1.0124, tangentMode=RCTM_User, time=0.3, value=1}
        {arriveTangent=-4.5105, interpMode=RCIM_Cubic, leaveTangent=-4.5105, tangentMode=RCTM_User, time=1, value=0}
       ]}
      YCurve {keys [
        <at 0 1.5>
        {interpMode=RCIM_Cubic, tangentMode=RCTM_User, time=0.3, value=1}
        {arriveTangent=-4.6311, interpMode=RCIM_Cubic, leaveTangent=-4.6311, tangentMode=RCTM_User, time=1, value=0}
       ]}
     }}
    Scale Sprite Size Mode=::NewEnumerator3
   )
   &ScaleColor(
    Scale RGB=&VectorFromFloat(Value=&FloatFromCurve(
      FloatCurve=DataInterface() {propertyValues {Curve {keys [<at 0 10>,<at 0.2 3>,<at 1 2>]}}}
      Scale Curve=&FN_FrontEndFloatModifier(Gameplay=10, Front End=1)
     ))
    ScaleA=true
    ScaleRGB=true
   )
  ]
  rendererClasses [&NiagaraSpriteRendererProperties]
  renderers [
   {rendererClass=&NiagaraSpriteRendererProperties}
   {rendererClass=&NiagaraSpriteRendererProperties, rendererIndex=1}
  ]
 }
 Spark_Burst001 {
  emitterUpdateScript [
   &EmitterState(Life Cycle Mode=::NewEnumerator1, Loop Behavior=::NewEnumerator1, Loop Delay=$User.EngineDelay, Loop Duration=1, UseLoopDelay=true)
   &SpawnRate(
    SpawnRate=&FloatFromCurve(
     FloatCurve=DataInterface() {propertyValues {Curve {keys [{interpMode=RCIM_Cubic, tangentMode=RCTM_User, time=0, value=1}, {interpMode=RCIM_Cubic, tangentMode=RCTM_User, time=1, value=0}]}}}
     CurveIndex=$Emitter.NormalizedLoopAge
     Scale Curve=20
    )
    Use Spawn Probability=true
   )
   &SpawnBurst_Instantaneous(Spawn Count=20, Use Spawn Probability=false, UseLoopCountLimit=false)
  ]
  particleSpawnScript [
   &FN_SetSpawnGroup(Number of Spawn Groups=1)
   &InitializeParticle(
    Color Maximum=<rgba 10 4.6167 0.5 1>
    Color Minimum=<rgba 10 3.8911 0.6366 1>
    Color Mode=::NewEnumerator2
    Lifetime Max=0.42
    Lifetime Min=0.15
    Lifetime Mode=::NewEnumerator1
    Sprite Size Max=<0.8 5.5>
    Sprite Size Min=<0.6 4>
    Sprite Size Mode=::NewEnumerator2
    UsePositionOffset=false
    Write Material Random=true
   )
   &SkeletalMeshLocation(
    Apply Sampled Orientation to Mesh Particles=true
    Bone / Socket Index=2
    Inherit Velocity=false
    Interpolate Mesh Sampling=false
    MaskSpawnGroup=false
    Mesh Sampling Type=::NewEnumerator5
    Sampled Position Offset Coordinate Space:ENiagaraMeshTransforms=::NewEnumerator3
    Skeletal Mesh=$User.SkeletalMesh
    Socket Sampling Mode=::NewEnumerator5
   )
   &ShapeLocation(MaskSpawnGroup=false, Override Seed=false, Sphere Radius=20, UseOwnerScale=true)
   &AddVelocity(
    Cone Angle=50
    Cone Axis=&FN_FrontEndVector3Modifier(Gameplay=<0 -1 0.3>, Front End=<0 1 0.3>)
    Module.UseConeFalloff=false
    Module.UseVelDistribution=false
    Override Seed=false
    Velocity Mode=::NewEnumerator2
    Velocity Speed=&RandomRangeFloat(Minimum=400, Maximum=800)
   )
   /SetVariables(
    Particles.GlowColor=<rgba 0.2 0.05 0 1>
    Particles.GlowSize=&Multiply_Vector2D(A=$Particles.SpriteSize, B=<10 4>)
   ) {isSetParameters=true, module=SetVariables_#8}
  ]
  particleUpdateScript [
   &ParticleState()
   &CurlNoiseForce(Noise Frequency=200, Noise Strength=4000, Pan Noise=false)
   &Drag(Drag=&RandomRangeFloat(Minimum=5, Maximum=6), Use Linear Drag=true, Use Rotational Drag=false)
   &SolveForcesAndVelocity(Clamp Velocity=false, Limit Acceleration=false)
   &ScaleSpriteSize(
    Non-Uniform Curve Sprite Scale=DataInterface() {propertyValues {
      XCurve {keys [
        <at 0 1.5>
        {arriveTangent=-1.0124, interpMode=RCIM_Cubic, leaveTangent=-1.0124, tangentMode=RCTM_User, time=0.3, value=1}
        {arriveTangent=-4.5105, interpMode=RCIM_Cubic, leaveTangent=-4.5105, tangentMode=RCTM_User, time=1, value=0}
       ]}
      YCurve {keys [
        <at 0 1.5>
        {interpMode=RCIM_Cubic, tangentMode=RCTM_User, time=0.3, value=1}
        {arriveTangent=-4.6311, interpMode=RCIM_Cubic, leaveTangent=-4.6311, tangentMode=RCTM_User, time=1, value=0}
       ]}
     }}
    Scale Sprite Size Mode=::NewEnumerator3
   )
   &ScaleColor(
    Scale RGB=&VectorFromFloat(Value=&FloatFromCurve(
      FloatCurve=DataInterface() {propertyValues {Curve {keys [<at 0 10>,<at 0.2 3>,<at 1 2>]}}}
      Scale Curve=&FN_FrontEndFloatModifier(Gameplay=10, Front End=1)
     ))
    ScaleA=true
    ScaleRGB=true
   )
  ]
  rendererClasses [&NiagaraSpriteRendererProperties]
  renderers [
   {rendererClass=&NiagaraSpriteRendererProperties}
   {rendererClass=&NiagaraSpriteRendererProperties, rendererIndex=1}
  ]
 }
 Spark_Burst002 {
  emitterUpdateScript [
   &EmitterState(Life Cycle Mode=::NewEnumerator1, Loop Delay=$User.EngineDelay, Loop Duration=1, UseLoopDelay=true)
   &SpawnRate(
    Spawn Probability=&Lerp_Float(
     A=0
     B=1
     Alpha=&FN_FrontEndFloatModifier(Gameplay=$System.FN_Integrator001.Value, Front End=$User.FE Intergrator Value)
    )
    SpawnRate=8
    Use Spawn Probability=true
   )
  ]
  particleSpawnScript [
   &FN_SetSpawnGroup(Number of Spawn Groups=2)
   &InitializeParticle(
    Color Maximum=<rgba 10 4.6167 0.5 1>
    Color Minimum=<rgba 10 3.8911 0.6366 1>
    Color Mode=::NewEnumerator2
    Lifetime Max=0.1
    Lifetime Min=0.15
    Lifetime Mode=::NewEnumerator1
    Sprite Size Max=<0.8 5.5>
    Sprite Size Min=<0.6 4>
    Sprite Size Mode=::NewEnumerator2
    UsePositionOffset=false
    Write Material Random=true
   )
   &SkeletalMeshLocation(
    Apply Sampled Orientation to Mesh Particles=true
    Bone / Socket Index=&SelectIntFromArray(Direct Array Index=$Particles.SpawnGroup, Int Selection Array=DataInterface() {propertyValues {IntData [0,1]}})
    Inherit Velocity=false
    Interpolate Mesh Sampling=false
    MaskSpawnGroup=false
    Mesh Sampling Type=::NewEnumerator5
    Sampled Position Offset Coordinate Space:ENiagaraMeshTransforms=::NewEnumerator3
    Skeletal Mesh=$User.SkeletalMesh
    Socket Sampling Mode=::NewEnumerator5
   )
   &ShapeLocation(MaskSpawnGroup=false, Override Seed=false, Sphere Radius=10, UseOwnerScale=true)
   &AddVelocity(
    Cone Angle=50
    Cone Axis=&FN_FrontEndVector3Modifier(
     Gameplay=&SelectVectorFromArray(
      Direct Array Index=$Particles.SpawnGroup
      Vector Selection Array=DataInterface() {propertyValues {FloatData [<-0.3 -1 -0.4>, <0.3 -1 -0.4>]}}
     )
     Front End=&SelectVectorFromArray(Direct Array Index=$Particles.SpawnGroup, Vector Selection Array=DataInterface() {propertyValues {FloatData [<0.3 1 -0.4>, <-0.3 1 -0.4>]}})
    )
    Module.UseConeFalloff=false
    Module.UseVelDistribution=false
    Override Seed=false
    Velocity Mode=::NewEnumerator2
    Velocity Speed=&RandomRangeFloat(Minimum=400, Maximum=800)
   )
   /SetVariables(
    Particles.GlowColor=<rgba 0.2 0.05 0 1>
    Particles.GlowSize=&Multiply_Vector2D(A=$Particles.SpriteSize, B=<10 4>)
   ) {isSetParameters=true, module=SetVariables_#8}
  ]
  particleUpdateScript [
   &ParticleState()
   &CurlNoiseForce(Noise Frequency=200, Noise Strength=1000, Pan Noise=false)
   &Drag(Drag=&RandomRangeFloat(Minimum=7, Maximum=8), Use Linear Drag=true, Use Rotational Drag=false)
   &SolveForcesAndVelocity(Clamp Velocity=false, Limit Acceleration=false)
   &ScaleSpriteSize(
    Non-Uniform Curve Sprite Scale=DataInterface() {propertyValues {
      XCurve {keys [
        <at 0 1.5>
        {arriveTangent=-1.0124, interpMode=RCIM_Cubic, leaveTangent=-1.0124, tangentMode=RCTM_User, time=0.3, value=1}
        {arriveTangent=-4.5105, interpMode=RCIM_Cubic, leaveTangent=-4.5105, tangentMode=RCTM_User, time=1, value=0}
       ]}
      YCurve {keys [
        <at 0 1.5>
        {interpMode=RCIM_Cubic, tangentMode=RCTM_User, time=0.3, value=1}
        {arriveTangent=-4.6311, interpMode=RCIM_Cubic, leaveTangent=-4.6311, tangentMode=RCTM_User, time=1, value=0}
       ]}
     }}
    Scale Sprite Size Mode=::NewEnumerator3
   )
   &ScaleColor(
    Scale RGB=&VectorFromFloat(Value=&FloatFromCurve(FloatCurve=DataInterface() {propertyValues {Curve {keys [<at 0 10>,<at 0.2 3>,<at 1 2>]}}}, Scale Curve=10))
    ScaleA=true
    ScaleRGB=true
   )
  ]
  rendererClasses [&NiagaraSpriteRendererProperties]
  renderers [
   {rendererClass=&NiagaraSpriteRendererProperties}
   {rendererClass=&NiagaraSpriteRendererProperties, rendererIndex=1}
  ]
 }
 Spark_Burst003 {
  emitterUpdateScript [
   &EmitterState(Life Cycle Mode=::NewEnumerator1, Loop Delay=$User.EngineDelay, Loop Duration=1, UseLoopDelay=true)
   &SpawnRate(
    Spawn Probability=&Lerp_Float(
     A=0
     B=1
     Alpha=&FN_FrontEndFloatModifier(Gameplay=$System.FN_Integrator001.Value, Front End=$User.FE Intergrator Value)
    )
    SpawnRate=10
    Use Spawn Probability=true
   )
  ]
  particleSpawnScript [
   &FN_SetSpawnGroup(Number of Spawn Groups=1)
   &InitializeParticle(
    Color Maximum=<rgba 10 4.6167 0.5 1>
    Color Minimum=<rgba 10 3.8911 0.6366 1>
    Color Mode=::NewEnumerator2
    Lifetime Max=0.1
    Lifetime Min=0.15
    Lifetime Mode=::NewEnumerator1
    Sprite Size Max=<0.8 5.5>
    Sprite Size Min=<0.6 4>
    Sprite Size Mode=::NewEnumerator2
    UsePositionOffset=false
    Write Material Random=true
   )
   &SkeletalMeshLocation(
    Apply Sampled Orientation to Mesh Particles=true
    Bone / Socket Index=2
    Inherit Velocity=false
    Interpolate Mesh Sampling=false
    MaskSpawnGroup=false
    Mesh Sampling Type=::NewEnumerator5
    Sampled Position Offset Coordinate Space:ENiagaraMeshTransforms=::NewEnumerator3
    Skeletal Mesh=$User.SkeletalMesh
    Socket Sampling Mode=::NewEnumerator5
   )
   &ShapeLocation(MaskSpawnGroup=false, Override Seed=false, Sphere Radius=20, UseOwnerScale=true)
   &AddVelocity(
    Cone Angle=50
    Cone Axis=&FN_FrontEndVector3Modifier(Gameplay=<0 -1 0.3>, Front End=<0 1 0.3>)
    Module.UseConeFalloff=false
    Module.UseVelDistribution=false
    Override Seed=false
    Velocity Mode=::NewEnumerator2
    Velocity Speed=&RandomRangeFloat(Minimum=800, Maximum=1200)
   )
   /SetVariables(
    Particles.GlowColor=<rgba 0.2 0.05 0 1>
    Particles.GlowSize=&Multiply_Vector2D(A=$Particles.SpriteSize, B=<10 4>)
   ) {isSetParameters=true, module=SetVariables_#8}
  ]
  particleUpdateScript [
   &ParticleState()
   &CurlNoiseForce(Noise Frequency=200, Noise Strength=1000, Pan Noise=false)
   &Drag(Drag=&RandomRangeFloat(Minimum=7, Maximum=8), Use Linear Drag=true, Use Rotational Drag=false)
   &SolveForcesAndVelocity(Clamp Velocity=false, Limit Acceleration=false)
   &ScaleSpriteSize(
    Non-Uniform Curve Sprite Scale=DataInterface() {propertyValues {
      XCurve {keys [
        <at 0 1.5>
        {arriveTangent=-1.0124, interpMode=RCIM_Cubic, leaveTangent=-1.0124, tangentMode=RCTM_User, time=0.3, value=1}
        {arriveTangent=-4.5105, interpMode=RCIM_Cubic, leaveTangent=-4.5105, tangentMode=RCTM_User, time=1, value=0}
       ]}
      YCurve {keys [
        <at 0 1.5>
        {interpMode=RCIM_Cubic, tangentMode=RCTM_User, time=0.3, value=1}
        {arriveTangent=-4.6311, interpMode=RCIM_Cubic, leaveTangent=-4.6311, tangentMode=RCTM_User, time=1, value=0}
       ]}
     }}
    Scale Sprite Size Mode=::NewEnumerator3
   )
   &ScaleColor(
    Scale RGB=&VectorFromFloat(Value=&FloatFromCurve(FloatCurve=DataInterface() {propertyValues {Curve {keys [<at 0 10>,<at 0.2 3>,<at 1 2>]}}}, Scale Curve=10))
    ScaleA=true
    ScaleRGB=true
   )
  ]
  rendererClasses [&NiagaraSpriteRendererProperties]
  renderers [
   {rendererClass=&NiagaraSpriteRendererProperties}
   {rendererClass=&NiagaraSpriteRendererProperties, rendererIndex=1}
  ]
 }
 Spark_Line {
  emitterUpdateScript [
   &EmitterState(Life Cycle Mode=::NewEnumerator1, Loop Delay=$User.EngineDelay, Loop Duration=1, UseLoopDelay=true)
   &SpawnRate(
    Spawn Probability=&Lerp_Float(
     A=0
     B=1
     Alpha=&FN_FrontEndFloatModifier(Gameplay=$System.FN_Integrator001.Value, Front End=$User.FE Intergrator Value)
    )
    SpawnRate=&RandomRangeFloat(Minimum=20, Maximum=30)
    Use Spawn Probability=true
   )
  ]
  particleSpawnScript [
   &FN_SetSpawnGroup(Number of Spawn Groups=2)
   &InitializeParticle(
    Color Maximum=<rgba 10 4.6167 0.5 1>
    Color Minimum=<rgba 10 3.8911 0.6366 1>
    Color Mode=::NewEnumerator2
    Lifetime Max=0.1
    Lifetime Min=0.12
    Lifetime Mode=::NewEnumerator1
    Sprite Size Max=<0.7 50>
    Sprite Size Min=<0.5 30>
    Sprite Size Mode=::NewEnumerator2
    UsePositionOffset=false
    Write Material Random=true
   )
   &AddVelocity(
    Velocity=&FN_FrontEndVector3Modifier(
     Gameplay=&SelectVectorFromArray(
      Direct Array Index=$Particles.SpawnGroup
      Vector Selection Array=DataInterface() {propertyValues {FloatData [<-0.3 -1 -0.4>, <0.3 -1 -0.4>]}}
     )
     Front End=&SelectVectorFromArray(Direct Array Index=$Particles.SpawnGroup, Vector Selection Array=DataInterface() {propertyValues {FloatData [<0.3 1 -0.5>, <-0.3 1 -0.5>]}})
    )
    Velocity Speed Scale=&FN_FrontEndFloatModifier(Gameplay=500, Front End=200)
   )
   &SkeletalMeshLocation(
    Apply Sampled Orientation to Mesh Particles=true
    Bone / Socket Index=&SelectIntFromArray(Direct Array Index=$Particles.SpawnGroup, Int Selection Array=DataInterface() {propertyValues {IntData [0,1]}})
    Inherit Velocity=false
    Interpolate Mesh Sampling=false
    MaskSpawnGroup=false
    Mesh Sampling Type=::NewEnumerator5
    Sampled Position Offset Coordinate Space:ENiagaraMeshTransforms=::NewEnumerator1
    Skeletal Mesh=$User.SkeletalMesh
    Socket Sampling Mode=::NewEnumerator5
   )
   &ShapeLocation(MaskSpawnGroup=false, Override Seed=false, Sphere Radius=8, UseOwnerScale=true)
   /SetVariables(
    Particles.GlowColor=<rgba 0.2 0.05 0 1>
    Particles.GlowSize=&Multiply_Vector2D(
     A=$Particles.SpriteSize
     B=&FN_FrontEndVector2Modifier(Front End:Vector2f=<10 2>, Gameplay:Vector2f=<10 6>)
    )
   ) {isSetParameters=true, module=SetVariables_#9}
  ]
  particleUpdateScript [
   &ParticleState()
   &SolveForcesAndVelocity(Clamp Velocity=false, Limit Acceleration=false)
   &ScaleSpriteSize(
    Non-Uniform Curve Scale=&FN_FrontEndVector2Modifier(Front End:Vector2f=<1 0.5>, Gameplay:Vector2f=<1 1>)
    Non-Uniform Curve Sprite Scale=DataInterface() {propertyValues {
      XCurve {keys [
        <at 0 1.5>
        {interpMode=RCIM_Cubic, tangentMode=RCTM_User, time=0.3, value=1}
        {arriveTangent=-0.9193, interpMode=RCIM_Cubic, leaveTangent=-0.9193, tangentMode=RCTM_User, time=1, value=0.5}
       ]}
      YCurve {keys [<at 0 1.5>, {arriveTangent=-2.3332, interpMode=RCIM_Cubic, leaveTangent=-2.3332, tangentMode=RCTM_User, time=1, value=0}]}
     }}
    Scale Sprite Size Mode=::NewEnumerator3
   )
   &ScaleColor(
    Scale RGB=&VectorFromFloat(Value=&FloatFromCurve(FloatCurve=DataInterface() {propertyValues {Curve {keys [<at 0 10>,<at 0.2 1>,<at 1 0.2>]}}}, Scale Curve=10))
    ScaleA=true
    ScaleRGB=true
   )
  ]
  rendererClasses [&NiagaraSpriteRendererProperties]
  renderers [
   {rendererClass=&NiagaraSpriteRendererProperties}
   {rendererClass=&NiagaraSpriteRendererProperties, rendererIndex=1}
  ]
 }
 Spark_Line_02 {
  emitterUpdateScript [
   &EmitterState(Life Cycle Mode=::NewEnumerator1, Loop Delay=$User.EngineDelay, Loop Duration=1, UseLoopDelay=true)
   &SpawnRate(
    Spawn Probability=&Lerp_Float(
     A=0
     B=1
     Alpha=&FN_FrontEndFloatModifier(Gameplay=$System.FN_Integrator001.Value, Front End=$User.FE Intergrator Value)
    )
    SpawnRate=&RandomRangeFloat(Minimum=30, Maximum=40)
    Use Spawn Probability=true
   )
  ]
  particleSpawnScript [
   &FN_SetSpawnGroup(Number of Spawn Groups=1)
   &InitializeParticle(
    Color Maximum=<rgba 10 4.6167 0.5 1>
    Color Minimum=<rgba 10 3.8911 0.6366 1>
    Color Mode=::NewEnumerator2
    Lifetime Max=0.1
    Lifetime Min=0.15
    Lifetime Mode=::NewEnumerator1
    Sprite Size Max=<0.7 50>
    Sprite Size Min=<0.5 50>
    Sprite Size Mode=::NewEnumerator2
    UsePositionOffset=false
    Write Material Random=true
   )
   &AddVelocity(
    Velocity=&FN_FrontEndVector3Modifier(Gameplay=<0 -1 0.3>, Front End=<0 1 0.3>)
    Velocity Speed Scale=&FN_FrontEndFloatModifier(Gameplay=1000, Front End=400)
   )
   &SkeletalMeshLocation(
    Apply Sampled Orientation to Mesh Particles=true
    Bone / Socket Index=2
    Inherit Velocity=false
    Interpolate Mesh Sampling=false
    MaskSpawnGroup=false
    Mesh Sampling Type=::NewEnumerator5
    Sampled Position Offset Coordinate Space:ENiagaraMeshTransforms=::NewEnumerator1
    Skeletal Mesh=$User.SkeletalMesh
    Socket Sampling Mode=::NewEnumerator5
   )
   &ShapeLocation(MaskSpawnGroup=false, Override Seed=false, Sphere Radius=12, UseOwnerScale=true)
   /SetVariables(
    Particles.GlowColor=<rgba 0.2 0.05 0 1>
    Particles.GlowSize=&Multiply_Vector2D(
     A=$Particles.SpriteSize
     B=&FN_FrontEndVector2Modifier(Front End:Vector2f=<10 2>, Gameplay:Vector2f=<10 6>)
    )
   ) {isSetParameters=true, module=SetVariables_#9}
  ]
  particleUpdateScript [
   &ParticleState()
   &SolveForcesAndVelocity(Clamp Velocity=false, Limit Acceleration=false)
   &ScaleSpriteSize(
    Non-Uniform Curve Scale=&FN_FrontEndVector2Modifier(Front End:Vector2f=<1 0.5>, Gameplay:Vector2f=<1 1>)
    Non-Uniform Curve Sprite Scale=DataInterface() {propertyValues {
      XCurve {keys [
        <at 0 1.5>
        {interpMode=RCIM_Cubic, tangentMode=RCTM_User, time=0.3, value=1}
        {arriveTangent=-0.9193, interpMode=RCIM_Cubic, leaveTangent=-0.9193, tangentMode=RCTM_User, time=1, value=0.5}
       ]}
      YCurve {keys [<at 0 1.5>, {arriveTangent=-2.3332, interpMode=RCIM_Cubic, leaveTangent=-2.3332, tangentMode=RCTM_User, time=1, value=0}]}
     }}
    Scale Sprite Size Mode=::NewEnumerator3
   )
   &ScaleColor(
    Scale RGB=&VectorFromFloat(Value=&FloatFromCurve(FloatCurve=DataInterface() {propertyValues {Curve {keys [<at 0 10>,<at 0.2 1>,<at 1 0.2>]}}}, Scale Curve=10))
    ScaleA=true
    ScaleRGB=true
   )
  ]
  rendererClasses [&NiagaraSpriteRendererProperties]
  renderers [
   {rendererClass=&NiagaraSpriteRendererProperties}
   {rendererClass=&NiagaraSpriteRendererProperties, rendererIndex=1}
  ]
 }
 SpeedLine001 {
  simTarget=GPUComputeSim
  emitterUpdateScript [
   &EmitterState(
    Delay First Loop Only=true
    Inactive Response=::NewEnumerator1
    Life Cycle Mode=::NewEnumerator1
    Loop Delay=$User.EngineDelay
    Loop Duration=3
    UseLoopDelay=true
   )
   &SpawnRate(
    SpawnRate=&FN_FrontEndFloatModifier(
     Gameplay=&Multiply_Float(
      A=&Multiply_Float(A=200, B=1)
      B=&Lerp_Float(A=-0.8, B=1, Alpha=$System.FN_Integrator001.Value)
     )
     Front End=60
    )
    Use Spawn Probability=false
   )
  ]
  particleSpawnScript [
   &InitializeParticle(
    Color=<rgba 1 1 1 10>
    Lifetime Max=0.12
    Lifetime Min=0.1
    Lifetime Mode=::NewEnumerator1
    Position Offset=&FN_FrontEndVector3Modifier(Gameplay=<0 0 0>, Front End=<0 0 0>)
    Sprite Size Max=<2 7>
    Sprite Size Min=<2 2>
    Sprite Size Mode=::NewEnumerator2
    UsePositionOffset=true
    Write Material Random=true
   )
   &TorusLocation(
    Align Torus With Axis=true
    Handle Radius=0
    Large Radius:NiagaraFloat=&RandomRangeFloat(
     Minimum=&FN_FrontEndFloatModifier(Gameplay=500, Front End=150)
     Maximum=&FN_FrontEndFloatModifier(Gameplay=600, Front End=150)
    )
    MaskSpawnGroup=false
    Module.Override Randomness=false
    Offset=&FN_FrontEndVector3Modifier(Gameplay=<0 100 0>, Front End=<0 -100 0>)
    Torus Axis=<0 1 0>
   )
   &AddVelocity(Velocity=&RandomRangeVector(Minimum=<0 500 0>, Maximum=<0 500 0>))
   &AddVelocityFromPoint(
    Module.Use Falloff Distance=false
    Origin Offset=&FN_FrontEndVector3Modifier(Gameplay=<0 200 0>, Front End=<0 200 0>)
    Velocity Strength=&FN_FrontEndFloatModifier(Gameplay=1500, Front End=300)
   )
   &CameraOffset(Camera Offset Amount=-10)
   &InitialMeshOrientation(Mesh Orientation Mode=::NewEnumerator2, Rotation=$User.RotationPreview, Use Rotation Vector=true)
  ]
  particleUpdateScript [
   &SolveForcesAndVelocity(Clamp Velocity=false, Limit Acceleration=false)
   &ParticleState()
   &ScaleSpriteSizeBySpeed(Max Scale Factor=<1 50>, Min Scale Factor=<0.01 0.05>, Velocity Threshold=1)
   &ScaleColor(
    Scale Alpha=&Multiply_Float(
     A=&FloatFromCurve(FloatCurve=DataInterface() {propertyValues {Curve {keys [
          {interpMode=RCIM_Cubic, tangentMode=RCTM_User, time=0, value=0}
          {interpMode=RCIM_Cubic, tangentMode=RCTM_User, time=0.5, value=1}
          {tangentMode=RCTM_User, time=1, value=0}
         ]}}})
     B=&Multiply_Float(
      A=&RandomRangeFloat(Minimum=0.1, Maximum=0.3)
      B=&FN_FrontEndFloatModifier(Gameplay=1, Front End=0)
     )
    )
    ScaleA=true
    ScaleRGB=true
   )
  ]
  rendererClasses [&NiagaraSpriteRendererProperties]
  renderers [{rendererClass=&NiagaraSpriteRendererProperties}]
 }
}
summary {
 systemName=NS_Glider_Inferno_Gold
 emitters [
  {emitterName=EngineFlames, rendererClasses [&NiagaraMeshRendererProperties]}
  {emitterName=Flare, rendererClasses [&NiagaraSpriteRendererProperties, &NiagaraLightRendererProperties]}
  {emitterName=EngineFlames_Burst, rendererClasses [&NiagaraMeshRendererProperties]}
  {emitterName=Flare_Burst, rendererClasses [&NiagaraSpriteRendererProperties]}
  {emitterName=Spark_Burst, rendererClasses [&NiagaraSpriteRendererProperties]}
  {emitterName=EmissiveMesh, rendererClasses [&NiagaraMeshRendererProperties]}
  {emitterName=Spark_Line, rendererClasses [&NiagaraSpriteRendererProperties]}
  {emitterName=Spark_Line_02, rendererClasses [&NiagaraSpriteRendererProperties]}
  {emitterName=Spark_Burst001, rendererClasses [&NiagaraSpriteRendererProperties]}
  {emitterName=Hair01, rendererClasses [&NiagaraMeshRendererProperties]}
  {emitterName=Hair_Flame, rendererClasses [&NiagaraMeshRendererProperties]}
  {emitterName=Hair_Thin, rendererClasses [&NiagaraMeshRendererProperties]}
  {emitterName=Spark_Burst002, rendererClasses [&NiagaraSpriteRendererProperties]}
  {emitterName=Spark_Burst003, rendererClasses [&NiagaraSpriteRendererProperties]}
  {emitterName=FlameMesh, rendererClasses [&NiagaraMeshRendererProperties]}
  {emitterName=FlameMesh001, rendererClasses [&NiagaraMeshRendererProperties, &NiagaraLightRendererProperties]}
  {emitterName=EngineFlames001, rendererClasses [&NiagaraMeshRendererProperties]}
  {emitterName=SpeedLine001, simTarget=GPUComputeSim, rendererClasses [&NiagaraSpriteRendererProperties]}
 ]
 userVariables [
  User.bIsFrontEnd {type=NiagaraBool}
  User.bIsFrontEndPreview {type=NiagaraBool}
  User.bIsFullyDeployed {type=NiagaraBool, defaultValue=true}
  User.EngineDelay {type=NiagaraFloat, defaultValue=0.16, description=""}
  User.FE Intergrator Value {type=NiagaraFloat, defaultValue=0.5, description=""}
  User.ForwardDot {type=NiagaraFloat, defaultValue=0, description=""}
  User.Interpolate_Rate_of_Change {type=NiagaraFloat, defaultValue=10, description=""}
  User.PlayerSpeed {type=NiagaraFloat, defaultValue=0, description=""}
  User.RibbonIDOffset {type=NiagaraInt32}
  User.RightDot {type=NiagaraFloat, defaultValue=0, description=""}
  User.RotationPreview {type=Vector3f, description="", defaultValue {struct=Vector3f, value=<0 0 0.5>}}
  User.SkeletalMesh {
   type=NiagaraDataInterfaceSkeletalMesh
   defaultValue {
    struct=NiagaraExt_VariableValue_DataInterface
    value {dataInterface=&self:NiagaraDataInterfaceSkeletalMesh_0}
   }
  }
 ]
}
